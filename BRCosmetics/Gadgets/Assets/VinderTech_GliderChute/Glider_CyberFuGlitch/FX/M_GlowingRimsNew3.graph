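# Translucent Default Lit material for glowing-rim Niagara mesh particles in the CyberFuGlitch glider's effects.
# A Fresnel mask, scaled by particle and vertex alpha and weighted by vertex channels, gets near-camera and depth fades.
# Emissive color blends a dynamic parameter with particle color by vertex red.
# Further branches build panning-noise refraction with mobile and quality fallbacks, plus texture-driven vertex offset controlled by dynamic parameters.
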Material /BRCosmetics/Gadgets/Assets/VinderTech_GliderChute/Glider_CyberFuGlitch/FX/M_GlowingRimsNew3
graph {
 properties {bUsedWithNiagaraMeshParticles=true, blendMode=BLEND_Translucent, pixelDepthOffsetMode=PDOM_Legacy, usedShadingModels=Default Lit}
 nodes [
  n1 = ParticleColor()
  n2 = VertexColor()
  n3 = Multiply(A=Multiply(A=@n1.A, B=@n2.A), B=Fresnel())
  n4 = Add(A=@n2.G, B=@n2.G)
  n5 = DynamicParameter() {parameterIndex=3, defaultValue=<rgba 0 1 1 1>, paramNames [R,G,B,LeftRight]}
  n6 = LinearInterpolate(A=@n2.B, B=@n2.B, Alpha=@n5.LeftRight)
  n7 = Multiply(A=@n3, B=Add(A=Multiply(A=@n4) {constB=0.5}, B=Multiply(A=@n6) {constB=3}))
  Add(A=@n3, B=@n7)
  Constant3Vector()
  Constant() {r=0.5}
  Constant() {r=0.5}
  LinearInterpolate() {constA=1, constB=0.98}
  LinearInterpolate(Alpha=Abs(None=Subtract(A=@n5.LeftRight) {constB=0.5})) {constA=1, constB=0.5}
  n8 = DynamicParameter() {defaultValue=<rgba 0 1 1 1>, paramNames [SepTrans, RefIntensity, "", Param4]}
  n9 = LinearInterpolate(Alpha=@n8.SepTrans) {constA=1.1, constB=1.035}
  n10 = Multiply(A=@n8.RefIntensity, B=ScalarParameter() {defaultValue=5, parameterName=RefIntensity})
  n11 = Power(Base=ComponentMask(None=TextureCoordinate()) {r=true}, Exp=ScalarParameter() {defaultValue=2, parameterName=RefFadeOut})
  n12 = MaterialFunctionCall(XY=TextureCoordinate() {coordinateIndex=1}) {materialFunction=&Swizzle}
  n13 = AppendVector(
   A=ScalarParameter() {defaultValue=1, group=Normal, parameterName=U_Tile_N}
   B=ScalarParameter() {defaultValue=0.25, group=Normal, parameterName=V_Tile_N}
  )
  n14 = TextureSampleParameter2D(UVs=Panner(Coordinate=Multiply(A=@n12.YX, B=@n13)) {speedY=-10}) {parameterName=RefTexture, texture=/Game/Effects/Fort_Effects/Textures/Tile/T_TileNoise_03}
  n15 = Constant() {r=1}
  n16 = ShadingPathSwitch(
   Default=LinearInterpolate(B=LinearInterpolate(B=@n9, Alpha=@n10) {constA=1, constB=1.1}, Alpha=Multiply(A=@n11, B=@n14.R)) {constA=1, constB=1.1}
   Mobile=@n15
  ) {desc=Migrated from FeatureLevelSwitch}
  n17 = VertexColor()
  n18 = Multiply(A=Add(A=@n17.G, B=@n17.G)) {constB=0.5}
  LinearInterpolate(
   B=LinearInterpolate(A=QualitySwitch(Default=@n16, Low=@n15, Medium=@n15), Alpha=Add(A=@n18, B=@n17.A)) {constA=0.99, constB=0.99}
   Alpha=ScalarParameter() {parameterName=Refract}
  ) {constA=1}
  MaterialFunctionCall() {materialFunction=&DebugTimeSine}
  n19 = DynamicParameter() {parameterIndex=1, paramNames [Amplitude,Frequency,OffsetX,OffsetY]}
  n20 = WorldPosition()
  n21 = ObjectPositionWS()
  n22 = MaterialFunctionCall(Float3=Divide(A=Subtract(A=@n20.XYZ, B=@n21), B=@n19.Frequency) {constB=50}) {materialFunction=&BreakOutFloat3Components}
  n23 = Add(A=@n19.OffsetY, B=ScalarParameter() {group="9) Vertex Offset", parameterName=Noise Offset})
  n24 = TextureObjectParameter() {
   group="9) Vertex Offset"
   parameterName=Vertex Offset
   samplerType=SAMPLERTYPE_Normal
   sortPriority=1
   texture=/Game/Effects/Fort_Effects/Textures/T_DustDistortion
  }
  n25 = TextureSample(UVs=AppendVector(A=Add(A=@n19.OffsetX, B=@n22.R), B=Add(A=@n22.G, B=@n23)), Tex=@n24)
  n26 = DynamicParameter() {desc=Axis of influence, parameterIndex=2, paramNames [X,Y,Z,Axis]}
  n27 = Multiply(
   A=@n25.RGB
   B=MaterialFunctionCall(X=@n26.X, Y=@n26.X, Z=@n26.X) {materialFunction=&MakeFloat3}.Result
  )
  n28 = Multiply(A=@n27, B=DynamicParameter() {parameterIndex=1, paramNames [Amplitude,Frequency,OffsetX,OffsetY]}.Amplitude)
  n29 = StaticSwitchParameter(
   True=StaticSwitchParameter(True=Transform(None=@n28) {transformSourceType=TRANSFORMSOURCE_Local}, False=@n28) {defaultValue=true, group="9) Vertex Offset", parameterName=Local Space, sortPriority=2}
   False=Constant3Vector()
  ) {defaultValue=true, group="9) Vertex Offset", parameterName=Enable Vertex Offset}
  n30 = MaterialFunctionCall!(InVector=@n29) {materialFunction=/Game/Athena/Environments/Landscape/Foliage/MF_MakePreciseVec3}
  Multiply(
   A=LinearInterpolate(
    A=Constant() {r=1}
    B=VertexColor().R
    Alpha=ScalarParameter() {desc=Enables Vertex COlor R to mask WPO., group="9) Vertex Offset", parameterName=Vertex Color R Masks WPO}
   )
   B=ShadingPathSwitch(Default=@n30.Result, Deferred=@n30.Result, Mobile=@n29) {desc=Migrated from FeatureLevelSwitch}
  )
  n31 = DynamicParameter() {parameterIndex=1, paramNames [Amplitude,Frequency,OffsetX,OffsetY]}
  Multiply(A=Clamp(None=Divide(A=@n22.B, B=@n31.OffsetY) {constB=100, desc=Falloff})) {constB=3}
  Multiply(A=@n2.A)
  n32 = MaterialFunctionCall(Distance=ScalarParameter() {defaultValue=1000, parameterName=NearCamFadeDistance}) {materialFunction=&Near_Camera_Fade}
  n33 = LinearInterpolate(A=@n3, B=@n7)
  NamedRerouteDeclaration(None=DepthFade(Opacity=Multiply(A=@n33, B=@n32.Result) {constA=1}, FadeDistance=ScalarParameter() {defaultValue=80, parameterName=FadeDistance})) {name=Mask, nodeColor=<rgba 0.4062 0 1 1>}
  NamedRerouteDeclaration(None=LinearInterpolate(A=@n5.RGB, B=@n1.RGB, Alpha=@n2.R)) {name=Emissive, nodeColor=<rgba 0.7109 1 0 1>}
  NamedRerouteUsage()
  NamedRerouteUsage()
  NamedRerouteUsage()
  ParticlePositionWS()
  ParticleRandom()
  ShadingPathSwitch(Default=@n27, Mobile=@n25.RGB) {desc=Migrated from FeatureLevelSwitch}
 ]
}
