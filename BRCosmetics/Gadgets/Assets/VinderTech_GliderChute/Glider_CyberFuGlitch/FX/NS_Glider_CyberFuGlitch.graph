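# Niagara system for the CyberFuGlitch glider cosmetic, spawning glowing mesh ripples from the glider's skeletal mesh.
# Separate emitters cover movement-based ripples in gameplay, a constant-rate front-end variant, and a timed deploy burst.
# User steering dot products and system timelines drive ripple velocity, scale and color, while curves animate dynamic material parameters.
# Deploy ripples blend position, orientation and velocity toward the sampled socket; FN_FrontEndFloatModifier gates spawning by context.
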NiagaraSystem /BRCosmetics/Gadgets/Assets/VinderTech_GliderChute/Glider_CyberFuGlitch/FX/NS_Glider_CyberFuGlitch
dependencies {
 usedDataInterfaces [
  &NiagaraDataInterfaceCurve
  &NiagaraDataInterfaceArrayInt32
  &NiagaraDataInterfaceVectorCurve
  &NiagaraDataInterfaceColorCurve
  &NiagaraDataInterfaceVector2DCurve
 ]
 usedDynamicInputs [
  &Boolean_AndOperation
  &Boolean_OrOperation
  &SetBoolByFloatComparison
  &NormalizeVector
  &Add_Vector
 ]
 usedModules [
  &SystemState
  &Timeline
  &self:Set_Trail_Opacity_02
  &FN_Integrator
  &self:SystemSpawnScript.SystemScriptSource.SystemScriptGraph.NiagaraNodeAssignment_2.SetVariables_#1
  &InterpolateOverTime
  &self:SystemSpawnScript.SystemScriptSource.SystemScriptGraph.NiagaraNodeAssignment_0.SetVariables_#2
  &self:SystemSpawnScript.SystemScriptSource.SystemScriptGraph.NiagaraNodeAssignment_1.SetVariables_#3
 ]
}
emitters {
 ExampleOrlandoSpawnGroups {
  bEnabled=false
  emitterUpdateScript [
   &EmitterState()
   &SpawnRate(
    SpawnRate=&Lerp_Float(A=120, B=200, Alpha=&ClampFloat(Float=$User.ForwardDot))
    Use Spawn Probability=false
   )
  ]
  particleSpawnScript [
   &PartitionParticles()
   /SetVariables(Particles.MyCustomIndex:NiagaraInt32=&SelectIntFromArray(Direct Array Index=$Particles.PartitionParticles.Partition, Int Selection Array=DataInterface() {propertyValues {IntData [2,3]}})) {isSetParameters=true}
   &InitializeParticle(Color=<rgba 1 0.375 0.7591 1>, Lifetime=0.4, Sprite Size Mode=::NewEnumerator3, UsePositionOffset=false, Write Material Random=true)
   /SetVariables(
    Particles.FrontToBack:NiagaraFloat=&RandomRangeFloat(Minimum=0, Maximum=1)
    Particles.LeanLerp:NiagaraFloat=&Multiply_Float(
     A=$User.RightDot
     B=&Lerp_Float(A=-1, B=1, Alpha=&MakeFloatFromInt(Int=$Particles.PartitionParticles.Partition))
    )
   ) {isSetParameters=true}
   &SkeletalMeshLocation(
    Bone / Socket Index=$Particles.MyCustomIndex
    MaskSpawnGroup=false
    Mesh Sampling Type=::NewEnumerator5
    Skeletal Mesh=$User.SkeletalMesh
    Socket Sampling Mode=::NewEnumerator5
   )
   &Color(Color=&LerpLinearColors(StartColor=<rgba 1 0 0 1>, EndColor=<rgba 0 1 0 1>, LerpFactor=$System.NormalizedRightDot))
  ]
  particleUpdateScript [
   &InheritVelocity(
    Module.Limit Speed=true
    Velocity Limit=10000
    Velocity Scale=&Multiply_VectorByFloat(Vector=<-0.03 -0.03 -0.1>, Float=0.1)
    Velocity Source=&TransformVector(Vector=$Engine.Owner.Velocity, Source Space=::World, Destination Space=::Local)
   )
   &GravityForce(Gravity=<0 0 -900>)
   &ParticleState()
   &SolveForcesAndVelocity(Clamp Velocity=false, Limit Acceleration=false)
  ]
  rendererClasses [&NiagaraSpriteRendererProperties]
  renderers [{rendererClass=&NiagaraSpriteRendererProperties}]
 }
 MeshRipplesMainLocal {
  emitterUpdateScript [
   &EmitterState(Delay First Loop Only=true, Life Cycle Mode=::NewEnumerator1, Loop Delay=1, Loop Duration=2, UseLoopDelay=true)
   &SpawnPerUnit(
    Max Movement Threshold=1200
    Module.Use Max Movement Threshold=true
    Module.Use Movement Tolerance=true
    Movement Tolerance=2
    Spawn Spacing=200
    Use Spawn Probability=true
   ) {enabled=false}
   &SpawnPerUnit(
    Module.Use Max Movement Threshold=true
    Module.Use Movement Tolerance=true
    Movement Tolerance=50
    Spawn Probability:NiagaraFloat=&FN_FrontEndFloatModifier(Gameplay=1, Front End=0)
    Spawn Spacing=350
    Use Spawn Probability=true
   ) {enabled=false, module=SpawnPerUnit001}
   &SpawnRate(SpawnRate=4, Use Spawn Probability=false) {enabled=false}
   &SpawnPerUnit(
    Module.Use Max Movement Threshold=true
    Module.Use Movement Tolerance=true
    Spawn Probability:NiagaraFloat=&FN_FrontEndFloatModifier(Gameplay=1, Front End=0)
    Spawn Spacing:NiagaraFloat=&Lerp_Float(A=175, B=300, Alpha=$System.Timeline.Time)
    Use Spawn Probability=true
   ) {module=SpawnPerUnit002}
  ]
  particleSpawnScript [
   &InitializeParticle(
    Color=<rgba 1 1 1 5>
    Lifetime=&Lerp_Float(A=1.5, B=1, Alpha=$User.ForwardDot)
    Mesh Scale Mode=::NewEnumerator3
    Mesh Uniform Scale=1.1
    Position Offset=<0 -10 -3>
    Position Offset Coordinate Space=::Local
    UsePositionOffset=true
    Write Material Random=true
   )
   &InitialMeshOrientation(Mesh Orientation Mode=::NewEnumerator2, Orientation Coordinate Space=::Local, Use Rotation Vector=false) {enabled=false}
   /SetVariables(Particles.AlphaFadeIn:NiagaraFloat=&FloatFromCurve(
     FloatCurve=DataInterface() {propertyValues {Curve {keys [{tangentMode=RCTM_User, time=2.4, value=0}, {tangentMode=RCTM_User, time=3, value=0.1}]}}}
     CurveIndex=$Emitter.Age
    )) {isSetParameters=true}
   &SkeletalMeshLocation(
    Bone / Socket Index=4
    MaskSpawnGroup=false
    Mesh Sampling Type=::NewEnumerator5
    Orientation Sampling=::NewEnumerator1
    Position Sampling=::NewEnumerator2
    Sampled Position Offset=<0 0 5>
    Sampled Velocity Coordinate Space:ENiagaraCoordinateSpace=::World
    Skeletal Mesh=$User.SkeletalMesh
    Socket Sampling Mode=::NewEnumerator5
   ) {module=SkeletalMeshLocation001}
   &InheritVelocity(
    Inherited Velocity Amount Scale=<-0.03 -0.03 0>
    Inherited Velocity Speed Limit=10000
    Speed Limit=false
    UseSpeedThreshold=false
    Velocity Source=&TransformVector(Vector=$Particles.SkeletalMeshLocation001.SampledVelocity, Source Space=::World, Destination Space=::Local)
   )
   &AddVelocity(Velocity=&MakeVector(
     X=&Multiply_Float(A=&Multiply_Float(A=$User.RightDot, B=0.5), B=-20)
     Y=&Add_Float(
      A=&Lerp_Float(A=0, B=50, Alpha=$System.Timeline.Time)
      B=&Lerp_Float(
       A=0
       B=100
       Alpha=&Abs_Float(Float:NiagaraFloat=&Multiply_Float(A=$User.RightDot, B=0.5))
      )
     )
     Z=0
    ))
   &ApplyInitialForces()
  ]
  particleUpdateScript [
   &ParticleState()
   &ScaleMeshSize(Scale Factor=&MakeVector(
     X=&Lerp_Float(
      A=&FloatFromCurve(FloatCurve=DataInterface() {propertyValues {Curve {keys [{tangentMode=RCTM_User, time=0, value=1}, {tangentMode=RCTM_User, time=1, value=1.2}]}}})
      B=1
      Alpha=$User.ForwardDot
     )
     Y=&Lerp_Float(A=0.9, B=1, Alpha=$User.ForwardDot)
     Z=1
    )) {enabled=false, module=ScaleMeshSize002}
   &ScaleMeshSize(Scale Factor=&MakeVector(
     X=&Multiply_Float(
      A=&Lerp_Float(A=1, B=1.1, Alpha=$System.Timeline.Time)
      B=&Lerp_Float(A=0.9, B=1, Alpha=$System.Timeline002.Time)
     )
     Y=&Lerp_Float(A=0.8, B=0.9, Alpha=$System.Timeline.Time)
     Z=1
    )) {module=ScaleMeshSize001}
   &ScaleMeshSize(Scale Factor=&AddVectorToPosition(
     Position=<0.8 0.9 0.5>
     Vector=&VectorFromCurve(
      VectorCurve=DataInterface() {propertyValues {
        XCurve {keys [
          {interpMode=RCIM_Cubic, tangentMode=RCTM_User, time=0.2, value=0}
          {arriveTangent=1.25, interpMode=RCIM_Cubic, leaveTangent=1.25, time=0.8, value=0.8}
          {arriveTangent=0.3863, interpMode=RCIM_Cubic, leaveTangent=0.3863, tangentMode=RCTM_User, time=1, value=1}
         ]}
        YCurve {keys [
          {arriveTangent=1.1365, interpMode=RCIM_Cubic, leaveTangent=1.1365, tangentMode=RCTM_User, time=0, value=0}
          {arriveTangent=3.8993, interpMode=RCIM_Cubic, leaveTangent=3.8993, tangentMode=RCTM_User, time=1, value=2}
         ]}
        ZCurve {keys [{interpMode=RCIM_Cubic, tangentMode=RCTM_User, time=0, value=0}, {interpMode=RCIM_Cubic, tangentMode=RCTM_User, time=1, value=0.2}]}
       }}
      Scale Curve=<0.7 0.7 0.5>
     )
    ))
   &UpdateMeshOrientation(Rotation Vector=&Lerp_Vector(
     A=<0 0 0>
     B=<0.02 0 0>
     Alpha=&Multiply_Float(
      A=$User.ForwardDot
      B=&FloatFromCurve(FloatCurve=DataInterface() {propertyValues {Curve {keys [{tangentMode=RCTM_User, time=0, value=0}, {tangentMode=RCTM_User, time=1, value=1}]}}})
     )
    )) {enabled=false, module=UpdateMeshOrientation002}
   &UpdateMeshOrientation(
    Facing Interpolation=false
    Orientation Method=::NewEnumerator2
    Override Quat=false
    TransformFacingOrigin=false
    TransformFacingPosition=false
    TransformUp=false
   ) {enabled=false}
   &SkeletalMeshLocation(
    Bone / Socket Index=4
    MaskSpawnGroup=false
    Mesh Sampling Type=::NewEnumerator5
    Position Sampling=::NewEnumerator2
    Sampled Orientation Coordinate Space=::Local
    Skeletal Mesh=$User.SkeletalMesh
    Socket Sampling Mode=::NewEnumerator5
   ) {enabled=false}
   &InheritSourceMovement(
    Applied Force Scale=&Multiply_VectorByFloat(Vector=<1 1 0>, Float=0.1)
    Applied Position Scale=&Multiply_VectorByFloat(Vector=<-0.03 -0.03 0>, Float=0.1)
    ApplyForce=true
    ApplyForceFalloff=false
    ApplyPosition=false
    Module.LimitForceSpeed=false
    Source Velocity:Vector3f=&TransformPosition(
     Position=Linked() {linkedVariable {name=Engine.Owner.Velocity, type=Vector3f}}
     Source Space=::World
     Destination Space=::Local
     Apply Scale=true
    )
   ) {enabled=false}
   &AccelerationForce(
    Acceleration=&MakeVector(
     X=&Multiply_Float(A=&Multiply_Float(A=$User.RightDot, B=0.5), B=-100)
     Y=&FloatFromCurve(
      FloatCurve=DataInterface() {propertyValues {Curve {keys [<at 0 100>,<at 0.5 0>,<at 1 500>]}}}
      CurveIndex=&Add_Float(A=&Multiply_Float(A=$User.ForwardDot, B=0.5), B=0.5)
     )
     Z=&FloatFromCurve(
      FloatCurve=DataInterface() {propertyValues {Curve {keys [
          {arriveTangent=0.4496, interpMode=RCIM_Cubic, leaveTangent=0.4496, tangentMode=RCTM_User, time=0, value=0}
          {arriveTangent=1.4269, interpMode=RCIM_Cubic, leaveTangent=1.4269, tangentMode=RCTM_User, time=1, value=1}
         ]}}}
      Scale Curve=80
     )
    )
    Coordinate Space=::Local
   )
   &DynamicMaterialParameters(
    Index 0 Param 2=&FloatFromCurve(
     FloatCurve=DataInterface() {propertyValues {Curve {keys [
         {arriveTangent=0.2665, interpMode=RCIM_Cubic, leaveTangent=0.2665, tangentMode=RCTM_User, time=0, value=1}
         {arriveTangent=-1, interpMode=RCIM_Cubic, leaveTangent=-1, time=0.3, value=0.925}
         {arriveTangent=-1.3379, interpMode=RCIM_Cubic, leaveTangent=-1.3379, tangentMode=RCTM_User, time=1, value=0}
        ]}}}
     Scale Curve=0.4
    )
    Index 1 Param 1=&FloatFromCurve(
     FloatCurve=DataInterface() {propertyValues {Curve {keys [
         {arriveTangent=0.7336, interpMode=RCIM_Cubic, leaveTangent=0.7336, tangentMode=RCTM_User, time=0, value=0}
         {arriveTangent=3.0033, interpMode=RCIM_Cubic, leaveTangent=3.0033, tangentMode=RCTM_User, time=0.4, value=4}
        ]}}}
     Scale Curve=4
    )
    Index 1 Param 2=150
    Index 1 Param 3=&Multiply_Float(A=$Emitter.Age, B=100)
    Index 1 Param 4=&FloatFromCurve(FloatCurve=DataInterface() {propertyValues {Curve {keys [<at 0 1>,<at 1 0>]}}})
    Index 2 Param 1=&FloatFromCurve(
     FloatCurve=DataInterface() {propertyValues {Curve {keys [
         {interpMode=RCIM_Cubic, tangentMode=RCTM_User, time=0, value=0}
         {interpMode=RCIM_Cubic, tangentMode=RCTM_User, time=0.3, value=1}
         {tangentMode=RCTM_User, time=1, value=0.5}
        ]}}}
     Scale Curve=0.5
    )
    Index 2 Param 2=&FloatFromCurve(
     FloatCurve=DataInterface() {propertyValues {Curve {keys [
         {interpMode=RCIM_Cubic, tangentMode=RCTM_User, time=0, value=0}
         {interpMode=RCIM_Cubic, tangentMode=RCTM_User, time=0.3, value=1}
         {tangentMode=RCTM_User, time=1, value=0.5}
        ]}}}
     Scale Curve=2
    )
    Index 3 Param 1=1
    Index 3 Param 2=1
    Index 3 Param 4=&FloatFromCurve(
     FloatCurve=DataInterface() {propertyValues {Curve {keys [
         {interpMode=RCIM_Cubic, tangentMode=RCTM_User, time=0, value=1}
         {arriveTangent=-3.5514, interpMode=RCIM_Cubic, leaveTangent=-3.5514, tangentMode=RCTM_User, time=1, value=0}
        ]}}}
     CurveIndex=$System.NormalizedRightDot
    )
    Module.Param0WriteEnabled=true
    Module.Param0WriteEnabled1=true
    Module.Param0WriteEnabled2=true
    Module.Param0WriteEnabled3=true
    Module.Param1WriteEnabled=true
    Module.Param1WriteEnabled1=true
    Module.Param1WriteEnabled2=true
    Module.Param1WriteEnabled3=true
    Module.Param2WriteEnabled=true
    Module.Param2WriteEnabled1=true
    Module.Param2WriteEnabled2=true
    Module.Param2WriteEnabled3=true
    Module.Param3WriteEnabled=true
    Module.Param3WriteEnabled1=true
    Module.Param3WriteEnabled2=true
    Module.Param3WriteEnabled3=true
    Write Parameter Index 1=true
    Write Parameter Index 2=true
    Write Parameter Index 3=true
   )
   &SolveForcesAndVelocity(Clamp Velocity=false, Limit Acceleration=false)
   &CameraOffset(Camera Offset Amount=-20)
   &ScaleColor(
    Scale Alpha=&Multiply_Float(
     A=&FloatFromCurve(FloatCurve=DataInterface() {propertyValues {Curve {keys [
          {arriveTangent=9.0725, interpMode=RCIM_Cubic, leaveTangent=9.0725, tangentMode=RCTM_User, time=0, value=0}
          {arriveTangent=-2.5249, interpMode=RCIM_Cubic, leaveTangent=-2.5249, tangentMode=RCTM_User, time=0.1408, value=0.9847}
          {arriveTangent=-0.0405, interpMode=RCIM_Cubic, leaveTangent=-0.0405, tangentMode=RCTM_User, time=1, value=0}
         ]}}})
     B=$Particles.AlphaFadeIn
    )
    Scale RGB=&MakeVectorFromLinearColorRGB(Color=&MultiplyLinearColor(
      Linear Color=&ColorFromCurve(ColorCurve=DataInterface() {propertyValues {
         AlphaCurve {keys [<at 0 1 RCIM_Cubic>,<at 1 1 RCIM_Cubic>]}
         BlueCurve {keys [<at 0.132 1 RCIM_Cubic>,<at 0.4893 1>,<at 0.8109 1 RCIM_Cubic>]}
         GreenCurve {keys [<at 0.132 0.6209 RCIM_Cubic>,<at 0.4893 0.3292>,<at 0.8109 0.0015 RCIM_Cubic>]}
         RedCurve {keys [<at 0.132 0.0534 RCIM_Cubic>,<at 0.4893 0.0023>,<at 0.8109 0 RCIM_Cubic>]}
        }})
      Scale Factor=&MakeFloat4FromVectorAndFloat(
       "Vector (RGB)"=&VectorFromFloat(Value=&FloatFromCurve(
         FloatCurve=DataInterface() {propertyValues {Curve {keys [
             {arriveTangent=-0.453, interpMode=RCIM_Cubic, leaveTangent=-0.453, tangentMode=RCTM_User, time=0.5, value=1}
             {arriveTangent=-3.3263, interpMode=RCIM_Cubic, leaveTangent=-3.3263, tangentMode=RCTM_User, time=1, value=0.2}
            ]}}}
         Scale Curve=200
        ))
       "Float (Alpha)"=1
      )
     ))
    ScaleA=true
    ScaleRGB=true
   )
   /SetVariables(Particles.LerpAttachedOrientation:NiagaraFloat=&FloatFromCurve(
     FloatCurve=DataInterface() {propertyValues {Curve {keys [{tangentMode=RCTM_User, time=0.2, value=0}, {tangentMode=RCTM_User, time=0.8, value=1}]}}}
     CurveIndex=$Emitter.NormalizedLoopAge
    )) {enabled=false, isSetParameters=true}
   /SetVariables(Particles.PositionLerped:NiagaraPosition=&LerpPosition(A=$Particles.Position, B=$Particles.SkeletalMeshLocation.SampledPosition, Alpha=$Particles.LerpAttachedOrientation)) {enabled=false, isSetParameters=true}
   /SetVariables(
    Particles.MeshOrientationLerped:Quat4f=&LerpQuaternion(
     Quaternion A:Quat4f=$Particles.MeshOrientation
     Quaternion B:Quat4f=&MultiplyQuaternion(
      Quaternion A=$Particles.SkeletalMeshLocation.SampledRotation
      Quaternion B=&MakeQuatFromAxisAngle(Axis=<0 0 1>, AngleInDegrees=90)
     )
     Lerp Factor:NiagaraFloat=$Particles.LerpAttachedOrientation
     Enable Spherical Linear Interpolation:NiagaraBool=false
    )
    Particles.VelocityLerped:Vector3f=&Lerp_Vector(A=$Particles.Velocity, B=$Particles.SkeletalMeshLocation.SampledVelocity, Alpha=$Particles.LerpAttachedOrientation)
   ) {enabled=false, isSetParameters=true}
   &UpdateMeshOrientation(
    Facing Coordinate Space:ENiagaraCoordinateSpace=::Local
    Facing Direction=$Particles.Velocity
    Facing Interpolation=false
    Orientation Method=::NewEnumerator1
    Override Quat=false
    TransformFacing=true
    TransformUp=false
   ) {enabled=false, module=UpdateMeshOrientation001}
   &ScaleColor(ScaleA=true, ScaleRGB=true) {enabled=false, module=ScaleColor001}
   /SetVariables(
    Particles.CameraOffset2:NiagaraFloat=&Multiply_Float(A=$Particles.CameraOffset, B=1)
    Particles.Color2=&MultiplyLinearColor(
     Linear Color=&LerpLinearColors(
      StartColor=&ColorFromCurve(ColorCurve=DataInterface() {propertyValues {
         AlphaCurve {keys [<at 0 1 RCIM_Cubic>,<at 1 1 RCIM_Cubic>]}
         BlueCurve {keys [<at 0.1443 1 RCIM_Cubic>,<at 0.4893 1>,<at 0.8109 1 RCIM_Cubic>]}
         GreenCurve {keys [<at 0.1443 0 RCIM_Cubic>,<at 0.4893 0.0652>,<at 0.8109 0.0826 RCIM_Cubic>]}
         RedCurve {keys [<at 0.1443 0.656 RCIM_Cubic>,<at 0.4893 0.3115>,<at 0.8109 0.3508 RCIM_Cubic>]}
        }})
      EndColor=&ColorFromCurve(ColorCurve=DataInterface() {propertyValues {
         AlphaCurve {keys [<at 0 1 RCIM_Cubic>,<at 1 1 RCIM_Cubic>]}
         BlueCurve {keys [<at 0.3066 1 RCIM_Cubic>, {arriveTangent=-1.6736, interpMode=RCIM_Cubic, leaveTangent=-1.6736, time=0.506, value=1}, <at 0.7846 0.2>]}
         GreenCurve {keys [<at 0.3066 0.2436 RCIM_Cubic>, {arriveTangent=-0.5096, interpMode=RCIM_Cubic, leaveTangent=-0.5096, time=0.506, value=0}, <at 0.7846 0>]}
         RedCurve {keys [<at 0.3066 0 RCIM_Cubic>, {arriveTangent=0.1552, interpMode=RCIM_Cubic, leaveTangent=0.1552, time=0.506, value=0.9}, <at 0.7846 0.0742>]}
        }})
      LerpFactor=$User.ForwardDot
     )
     Scale Factor=&MakeFloat4FromVectorAndFloat(
      "Vector (RGB)"=&VectorFromFloat(Value=&FloatFromCurve(FloatCurve=DataInterface() {propertyValues {Curve {keys [
            {arriveTangent=0.0734, interpMode=RCIM_Cubic, leaveTangent=0.0734, tangentMode=RCTM_User, time=0, value=1}
            {interpMode=RCIM_Cubic, tangentMode=RCTM_User, time=1, value=1}
           ]}}}))
      "Float (Alpha)"=&Multiply_Float(
       A=&Lerp_Float(A=15, B=30, Alpha=$System.Timeline.Time)
       B=&MakeFloatFromLinearColor(LinearColor=$Particles.Color, Channel:ENiagaraLinearColor_Channels=::NewEnumerator3)
      )
     )
    )
    Particles.Position2:NiagaraPosition=&AddVectorToPosition(
     Position=$Particles.Position
     Vector=&MakeVector(
      X=0
      Y=0
      Z=&FloatFromCurve(
       FloatCurve=DataInterface() {propertyValues {Curve {keys [
           {interpMode=RCIM_Cubic, tangentMode=RCTM_User, time=0, value=-5}
           {interpMode=RCIM_Cubic, tangentMode=RCTM_User, time=0.5, value=5}
           {tangentMode=RCTM_User, time=1, value=-3}
          ]}}}
       Scale Curve=0.4
      )
     )
    )
    Particles.Scale2:Vector3f=&Multiply_Vector(A=$Particles.Scale, B=<1 1 0.8>)
   ) {isSetParameters=true}
   &ScaleColor(Scale Alpha=&Lerp_Float(A=0.2, B=1, Alpha=$System.Timeline.Time), ScaleA=true, ScaleRGB=true) {module=ScaleColor002}
  ]
  rendererClasses [&NiagaraSpriteRendererProperties, &NiagaraMeshRendererProperties]
  renderers [
   {rendererClass=&NiagaraSpriteRendererProperties}
   {rendererClass=&NiagaraMeshRendererProperties, rendererIndex=1}
   {rendererClass=&NiagaraMeshRendererProperties, rendererIndex=2}
  ]
 }
 MeshRipples_FE {
  emitterUpdateScript [
   &EmitterState(Delay First Loop Only=true, Life Cycle Mode=::NewEnumerator1, Loop Delay=1, Loop Duration=1, UseLoopDelay=true)
   &SpawnRate(
    Spawn Probability=&FN_FrontEndFloatModifier(Gameplay=0, Front End=1)
    SpawnRate=8
    Use Spawn Probability=true
   )
  ]
  particleSpawnScript [
   &InitializeParticle(
    Color=&MakeLinearColorFromVectorAndFloat("Vector (RGB)"=<1 1 1>, "Float (Alpha)"=&RandomRangeFloat(Minimum=0.1, Maximum=5))
    Mesh Scale Mode=::NewEnumerator3
    Position Mode=::NewEnumerator1
    UsePositionOffset=false
    Write Material Random=true
   )
   &SkeletalMeshLocation(
    Bone / Socket Index=4
    MaskSpawnGroup=false
    Mesh Sampling Type=::NewEnumerator5
    Sampled Position Offset=<5 0 10>
    Skeletal Mesh=$User.SkeletalMesh
    Socket Sampling Mode=::NewEnumerator5
   ) {module=SkeletalMeshLocation001}
   /SetVariables(Particles.AlphaFadeIn:NiagaraFloat=&FloatFromCurve(
     FloatCurve=DataInterface() {propertyValues {Curve {keys [<at 1 0 RCIM_Cubic>, {arriveTangent=0.0793, interpMode=RCIM_Cubic, leaveTangent=0.0793, tangentMode=RCTM_User, time=3, value=0.1}]}}}
     CurveIndex=$Emitter.Age
    )) {isSetParameters=true}
   &InheritVelocity(
    Inherited Velocity Amount Scale=<-0.03 -0.03 0>
    Inherited Velocity Speed Limit=10000
    Speed Limit=false
    UseSpeedThreshold=false
    Velocity Source=&TransformVector(Vector=$Particles.SkeletalMeshLocation001.SampledVelocity, Source Space=::World, Destination Space=::Local)
   )
   &AddVelocity(Velocity=<35 500 0>)
   &ApplyInitialForces()
  ]
  particleUpdateScript [
   &ParticleState()
   &UpdateMeshOrientation(Rotation Vector=<0 0.009 0>)
   &ScaleMeshSize(Scale Factor=&AddVectorToPosition(
     Position=<1 1 1>
     Vector=&VectorFromCurve(
      VectorCurve=DataInterface() {propertyValues {
        XCurve {keys [
          {arriveTangent=0.2304, interpMode=RCIM_Cubic, leaveTangent=0.2304, tangentMode=RCTM_User, time=0.0076, value=0.02}
          {arriveTangent=1.3485, interpMode=RCIM_Cubic, leaveTangent=1.3485, tangentMode=RCTM_User, time=0.5424, value=0.4103}
          {arriveTangent=1.9636, interpMode=RCIM_Cubic, leaveTangent=1.9636, tangentMode=RCTM_User, time=1, value=1.2}
         ]}
        YCurve {keys [
          {arriveTangent=1.4463, interpMode=RCIM_Cubic, leaveTangent=1.4463, tangentMode=RCTM_User, time=0.2, value=0}
          {arriveTangent=-3.18, interpMode=RCIM_Cubic, leaveTangent=-3.18, tangentMode=RCTM_User, time=0.9878, value=0.2}
         ]}
        ZCurve {keys [{interpMode=RCIM_Cubic, tangentMode=RCTM_User, time=0, value=0}, {interpMode=RCIM_Cubic, tangentMode=RCTM_User, time=1, value=0.2}]}
       }}
      Scale Curve=<3 2 1>
     )
    ))
   &CameraOffset(Camera Offset Amount=&FloatFromCurve(FloatCurve=DataInterface() {propertyValues {Curve {keys [<at 0 -50>,<at 1 -20>]}}}, CurveIndex=$Emitter.Age))
   &ScaleColor(
    Scale Alpha=&Multiply_Float(
     A=&FloatFromCurve(FloatCurve=DataInterface() {propertyValues {Curve {keys [
          {arriveTangent=0.5165, interpMode=RCIM_Cubic, leaveTangent=0.5165, tangentMode=RCTM_User, time=0, value=0}
          {arriveTangent=-2.5249, interpMode=RCIM_Cubic, leaveTangent=-2.5249, tangentMode=RCTM_User, time=0.2, value=0.9847}
          {arriveTangent=-0.0405, interpMode=RCIM_Cubic, leaveTangent=-0.0405, tangentMode=RCTM_User, time=0.6, value=0.1}
          <at 1 0>
         ]}}})
     B=$Particles.AlphaFadeIn
    )
    Scale RGB=&MakeVectorFromLinearColorRGB(Color=&MultiplyLinearColor(
      Linear Color=&ColorFromCurve(ColorCurve=DataInterface() {propertyValues {
         AlphaCurve {keys [<at 0 1 RCIM_Cubic>,<at 1 1 RCIM_Cubic>]}
         BlueCurve {keys [<at 0.132 1 RCIM_Cubic>,<at 0.4893 1>,<at 0.8109 1 RCIM_Cubic>]}
         GreenCurve {keys [<at 0.132 0.6209 RCIM_Cubic>,<at 0.4893 0.3292>,<at 0.8109 0.0015 RCIM_Cubic>]}
         RedCurve {keys [<at 0.132 0.0534 RCIM_Cubic>,<at 0.4893 0.0023>,<at 0.8109 0 RCIM_Cubic>]}
        }})
      Scale Factor=&MakeFloat4FromVectorAndFloat(
       "Vector (RGB)"=&VectorFromFloat(Value=&FloatFromCurve(
         FloatCurve=DataInterface() {propertyValues {Curve {keys [
             {arriveTangent=-0.453, interpMode=RCIM_Cubic, leaveTangent=-0.453, tangentMode=RCTM_User, time=0.5, value=1}
             {arriveTangent=-3.3263, interpMode=RCIM_Cubic, leaveTangent=-3.3263, tangentMode=RCTM_User, time=1, value=0.2}
            ]}}}
         Scale Curve=100
        ))
       "Float (Alpha)"=1
      )
     ))
    ScaleA=true
    ScaleRGB=true
   )
   &DynamicMaterialParameters(
    Index 0 Param 1=1
    Index 0 Param 2=&FloatFromCurve(
     FloatCurve=DataInterface() {propertyValues {Curve {keys [
         {arriveTangent=0.2665, interpMode=RCIM_Cubic, leaveTangent=0.2665, tangentMode=RCTM_User, time=0, value=1}
         {arriveTangent=-1, interpMode=RCIM_Cubic, leaveTangent=-1, time=0.3, value=0.925}
         {arriveTangent=-1.3379, interpMode=RCIM_Cubic, leaveTangent=-1.3379, tangentMode=RCTM_User, time=1, value=0}
        ]}}}
     Scale Curve=0
    )
    Index 1 Param 1=&FloatFromCurve(
     FloatCurve=DataInterface() {propertyValues {Curve {keys [
         {arriveTangent=0.7336, interpMode=RCIM_Cubic, leaveTangent=0.7336, tangentMode=RCTM_User, time=0, value=0}
         {arriveTangent=3.0033, interpMode=RCIM_Cubic, leaveTangent=3.0033, tangentMode=RCTM_User, time=0.4, value=4}
        ]}}}
     Scale Curve=4
    )
    Index 1 Param 2=150
    Index 1 Param 3=&Multiply_Float(A=$Emitter.Age, B=100)
    Index 1 Param 4=&FloatFromCurve(FloatCurve=DataInterface() {propertyValues {Curve {keys [<at 0 1>,<at 1 0>]}}})
    Index 2 Param 1=&FloatFromCurve(
     FloatCurve=DataInterface() {propertyValues {Curve {keys [
         {interpMode=RCIM_Cubic, tangentMode=RCTM_User, time=0, value=0}
         {interpMode=RCIM_Cubic, tangentMode=RCTM_User, time=0.3, value=1}
         {tangentMode=RCTM_User, time=1, value=0.5}
        ]}}}
     Scale Curve=0.5
    )
    Index 2 Param 2=&FloatFromCurve(
     FloatCurve=DataInterface() {propertyValues {Curve {keys [
         {interpMode=RCIM_Cubic, tangentMode=RCTM_User, time=0, value=0}
         {interpMode=RCIM_Cubic, tangentMode=RCTM_User, time=0.3, value=1}
         {tangentMode=RCTM_User, time=1, value=0.5}
        ]}}}
     Scale Curve=2
    )
    Index 3 Param 1=1
    Index 3 Param 2=1
    Index 3 Param 4=&FloatFromCurve(
     FloatCurve=DataInterface() {propertyValues {Curve {keys [
         {interpMode=RCIM_Cubic, tangentMode=RCTM_User, time=0, value=1}
         {arriveTangent=-3.5514, interpMode=RCIM_Cubic, leaveTangent=-3.5514, tangentMode=RCTM_User, time=1, value=0}
        ]}}}
     CurveIndex=$System.NormalizedRightDot
    )
    Module.Param0WriteEnabled=true
    Module.Param0WriteEnabled1=true
    Module.Param0WriteEnabled2=true
    Module.Param0WriteEnabled3=true
    Module.Param1WriteEnabled=true
    Module.Param1WriteEnabled1=true
    Module.Param1WriteEnabled2=true
    Module.Param1WriteEnabled3=true
    Module.Param2WriteEnabled=true
    Module.Param2WriteEnabled1=true
    Module.Param2WriteEnabled2=true
    Module.Param2WriteEnabled3=true
    Module.Param3WriteEnabled=true
    Module.Param3WriteEnabled1=true
    Module.Param3WriteEnabled2=true
    Module.Param3WriteEnabled3=true
    Write Parameter Index 1=true
    Write Parameter Index 2=true
    Write Parameter Index 3=true
   )
   /SetVariables(
    Particles.CameraOffset2:NiagaraFloat=&Multiply_Float(A=$Particles.CameraOffset, B=1)
    Particles.Color2=&MultiplyLinearColor(
     Linear Color=&ColorFromCurve(ColorCurve=DataInterface() {propertyValues {
        AlphaCurve {keys [<at 0 1 RCIM_Cubic>,<at 1 1 RCIM_Cubic>]}
        BlueCurve {keys [<at 0.0123 5>, {arriveTangent=-5.0088, interpMode=RCIM_Cubic, leaveTangent=-5.0088, time=0.2279, value=5}, <at 0.8109 1 RCIM_Cubic>]}
        GreenCurve {keys [
          <at 0.0123 0.3261>
          {arriveTangent=-0.305, interpMode=RCIM_Cubic, leaveTangent=-0.305, time=0.2279, value=0.5}
          <at 0.8109 0.0826 RCIM_Cubic>
         ]}
        RedCurve {keys [
          <at 0.0123 1.5576>
          {arriveTangent=-1.5112, interpMode=RCIM_Cubic, leaveTangent=-1.5112, time=0.2279, value=3.452}
          <at 0.8109 0.3508 RCIM_Cubic>
         ]}
       }})
     Scale Factor=&MakeFloat4FromVectorAndFloat(
      "Vector (RGB)"=&VectorFromFloat(Value=&FloatFromCurve(FloatCurve=DataInterface() {propertyValues {Curve {keys [
            {arriveTangent=0.0734, interpMode=RCIM_Cubic, leaveTangent=0.0734, tangentMode=RCTM_User, time=0, value=1}
            {interpMode=RCIM_Cubic, tangentMode=RCTM_User, time=1, value=1}
           ]}}}))
      "Float (Alpha)"=&Multiply_Float(
       A=&FloatFromCurve(FloatCurve=DataInterface() {propertyValues {Curve {keys [<at 0 1>,<at 1 0>]}}}, Scale Curve=10)
       B=&MakeFloatFromLinearColor(LinearColor=$Particles.Color, Channel:ENiagaraLinearColor_Channels=::NewEnumerator3)
      )
     )
    )
    Particles.Position2:NiagaraPosition=&AddVectorToPosition(Position=$Particles.Position, Vector=&MakeVector(X=0, Y=0, Z=0))
    Particles.Scale2:Vector3f=&Multiply_Vector(A=$Particles.Scale, B=<1 1 1>)
   ) {isSetParameters=true}
   &SkeletalMeshLocation(
    Bone / Socket Index=4
    MaskSpawnGroup=false
    Mesh Sampling Type=::NewEnumerator5
    Orientation Sampling=::NewEnumerator1
    Sampled Position Offset=&MakeVector(
     X=&FloatFromCurve(
      FloatCurve=DataInterface() {propertyValues {Curve {keys [{tangentMode=RCTM_User, time=0, value=0}, {tangentMode=RCTM_User, time=1, value=1}]}}}
      Scale Curve=0
     )
     Y=0
     Z=0
    )
    Skeletal Mesh=$User.SkeletalMesh
    Socket Sampling Mode=::NewEnumerator5
   ) {enabled=false}
   &SolveForcesAndVelocity(Clamp Velocity=false, Limit Acceleration=false)
  ]
  rendererClasses [&NiagaraSpriteRendererProperties, &NiagaraMeshRendererProperties]
  renderers [
   {rendererClass=&NiagaraSpriteRendererProperties}
   {rendererClass=&NiagaraMeshRendererProperties, rendererIndex=1}
   {rendererClass=&NiagaraMeshRendererProperties, rendererIndex=2}
   {rendererClass=&NiagaraMeshRendererProperties, rendererIndex=3}
  ]
 }
 Ripples_Deploy {
  emitterUpdateScript [
   &EmitterState(Life Cycle Mode=::NewEnumerator1, Loop Behavior=::NewEnumerator1, Loop Delay=0.2, Loop Duration=2.4, UseLoopDelay=true)
   &SpawnRate(
    Spawn Probability=&FN_FrontEndFloatModifier(Gameplay=1, Front End=0)
    SpawnRate=&FloatFromCurve(FloatCurve=DataInterface() {propertyValues {Curve {keys [<at 0.7 1>,<at 1 0>]}}}, CurveIndex=$Emitter.NormalizedLoopAge, Scale Curve=40)
    Use Spawn Probability=true
   )
  ]
  particleSpawnScript [
   &InitializeParticle(
    Color=<rgba 1 1 1 0.1>
    Lifetime=&FloatFromCurve(FloatCurve=DataInterface() {propertyValues {Curve {keys [<at 0.5 0.2>,<at 0.9 0.55>,<at 1 1>]}}}, CurveIndex=$Emitter.NormalizedLoopAge)
    Mesh Scale Mode=::NewEnumerator3
    Position Mode=::NewEnumerator0
    Write Material Random=true
   )
   &InitialMeshOrientation(Mesh Axis To Orient=<0 -1 0>, Mesh Orientation Mode=::NewEnumerator2, Use Rotation Vector=false)
   /SetVariables(Particles.AlphaFadeIn:NiagaraFloat=&FloatFromCurve(
     FloatCurve=DataInterface() {propertyValues {Curve {keys [{tangentMode=RCTM_User, time=1.8, value=1}, {tangentMode=RCTM_User, time=2.4, value=0}]}}}
     CurveIndex=$Emitter.Age
    )) {isSetParameters=true, module=SetVariables_#4}
   &SkeletalMeshLocation(
    Bone / Socket Index=4
    MaskSpawnGroup=false
    Mesh Sampling Type=::NewEnumerator5
    Sampled Position Offset=<0 0 -8>
    Skeletal Mesh=$User.SkeletalMesh
    Socket Sampling Mode=::NewEnumerator5
   ) {module=SkeletalMeshLocation001}
   &InheritVelocity(
    Inherited Velocity Amount Scale=&MakeVector(
     X=0.9
     Y=0.9
     Z=&FloatFromCurve(FloatCurve=DataInterface() {propertyValues {Curve {keys [<at 0 1.1>,<at 1 0.9>]}}})
    )
    Inherited Velocity Speed Limit=10000
    Speed Limit=false
    UseSpeedThreshold=false
   )
   &InheritVelocity(
    Inherited Velocity Amount Scale=&MakeVector(X=1, Y=1, Z=1)
    Inherited Velocity Speed Limit=10000
    Speed Limit=false
    UseSpeedThreshold=false
   ) {enabled=false, module=InheritVelocity001}
   &AddVelocity(Velocity=&MakeVector(
     X=0
     Y=&Add_Float(
      A=&Lerp_Float(A=0, B=500, Alpha=$System.Timeline.Time)
      B=&Lerp_Float(
       A=0
       B=20
       Alpha=&Abs_Float(Float:NiagaraFloat=&Multiply_Float(A=$User.RightDot, B=0.5))
      )
     )
     Z=-2
    )) {enabled=false}
   &AddVelocity(Velocity=<0 1000 0>) {enabled=false, module=AddVelocity001}
   &ApplyInitialForces()
  ]
  particleUpdateScript [
   &ParticleState()
   &ScaleMeshSize(Scale Factor=&VectorFromCurve(
     VectorCurve=DataInterface() {propertyValues {
       XCurve {keys [
         <at 0.2 1.05 RCIM_Cubic>
         {arriveTangent=-0.3438, interpMode=RCIM_Cubic, leaveTangent=-0.3438, tangentMode=RCTM_User, time=1.5, value=0.7}
         <at 2.4 0.5>
        ]}
       YCurve {keys [
         <at 0.2 1.05 RCIM_Cubic>
         {arriveTangent=-0.3438, interpMode=RCIM_Cubic, leaveTangent=-0.3438, tangentMode=RCTM_User, time=1.5, value=0.7}
         <at 2.4 0.5>
        ]}
       ZCurve {keys [
         <at 0.2 1.05 RCIM_Cubic>
         {arriveTangent=-0.3438, interpMode=RCIM_Cubic, leaveTangent=-0.3438, tangentMode=RCTM_User, time=1.5, value=0.7}
         <at 2.4 0.5>
        ]}
      }}
     CurveIndex=$Emitter.Age
    )) {module=ScaleMeshSize004}
   &ScaleMeshSize(Scale Factor=&AddVectorToPosition(
     Position=<1 1 0.5>
     Vector=&VectorFromCurve(
      VectorCurve=DataInterface() {propertyValues {
        XCurve {keys [
          {arriveTangent=-1.2724, interpMode=RCIM_Cubic, leaveTangent=-1.2724, tangentMode=RCTM_User, time=0, value=1}
          {interpMode=RCIM_Cubic, tangentMode=RCTM_User, time=1, value=0}
         ]}
        YCurve {keys [
          {arriveTangent=-1.2724, interpMode=RCIM_Cubic, leaveTangent=-1.2724, tangentMode=RCTM_User, time=0, value=1}
          {interpMode=RCIM_Cubic, tangentMode=RCTM_User, time=1, value=0}
         ]}
        ZCurve {keys [
          {arriveTangent=-1.2724, interpMode=RCIM_Cubic, leaveTangent=-1.2724, tangentMode=RCTM_User, time=0, value=1}
          {interpMode=RCIM_Cubic, tangentMode=RCTM_User, time=1, value=0}
         ]}
       }}
      Scale Curve=<0.5 0.6 0.5>
     )
    )) {enabled=false, module=ScaleMeshSize001}
   &ScaleMeshSize(Scale Factor=&MakeVector(X=&Lerp_Float(A=1, B=1.069, Alpha=$System.Timeline.Time), Y=1, Z=1))
   &ScaleMeshSize(Scale Factor=&AddVectorToPosition(
     Position=<1.1 1.1 0.5>
     Vector=&VectorFromCurve(
      VectorCurve=DataInterface() {propertyValues {
        XCurve {keys [
          {interpMode=RCIM_Cubic, tangentMode=RCTM_User, time=0.2, value=0}
          {arriveTangent=1.25, interpMode=RCIM_Cubic, leaveTangent=1.25, time=0.8, value=0.8}
          {arriveTangent=0.3863, interpMode=RCIM_Cubic, leaveTangent=0.3863, tangentMode=RCTM_User, time=1, value=1}
         ]}
        YCurve {keys [
          {arriveTangent=-0.6221, interpMode=RCIM_Cubic, leaveTangent=-0.6221, tangentMode=RCTM_User, time=0, value=0.4}
          {arriveTangent=1.6, interpMode=RCIM_Cubic, leaveTangent=1.6, time=0.1, value=0.125}
          {arriveTangent=3.8993, interpMode=RCIM_Cubic, leaveTangent=3.8993, tangentMode=RCTM_User, time=1, value=2}
         ]}
        ZCurve {keys [{interpMode=RCIM_Cubic, tangentMode=RCTM_User, time=0, value=0}, {interpMode=RCIM_Cubic, tangentMode=RCTM_User, time=1, value=0.2}]}
       }}
      Scale Curve=<0.7 0.7 0.5>
     )
    )) {module=ScaleMeshSize002}
   &ScaleMeshSize(Scale Factor=&MakeVector(
     X=&Lerp_Float(
      A=&FloatFromCurve(FloatCurve=DataInterface() {propertyValues {Curve {keys [{tangentMode=RCTM_User, time=0, value=1}, {tangentMode=RCTM_User, time=1, value=1.2}]}}})
      B=1
      Alpha=$System.Timeline.Time
     )
     Y=&Lerp_Float(A=0.9, B=0.9, Alpha=$System.Timeline.Time)
     Z=1
    )) {module=ScaleMeshSize003}
   &ScaleColor(
    Scale Alpha=&Multiply_Float(
     A=&FloatFromCurve(FloatCurve=DataInterface() {propertyValues {Curve {keys [
          {arriveTangent=9.0725, interpMode=RCIM_Cubic, leaveTangent=9.0725, tangentMode=RCTM_User, time=0, value=0}
          {arriveTangent=-2.5249, interpMode=RCIM_Cubic, leaveTangent=-2.5249, tangentMode=RCTM_User, time=0.1408, value=0.9847}
          <at 0.2111 0.1382>
          {arriveTangent=-0.0405, interpMode=RCIM_Cubic, leaveTangent=-0.0405, tangentMode=RCTM_User, time=1, value=0}
         ]}}})
     B=$Particles.AlphaFadeIn
    )
    Scale RGB=&MakeVectorFromLinearColorRGB(Color=&MultiplyLinearColor(
      Linear Color=&ColorFromCurve(ColorCurve=DataInterface() {propertyValues {
         AlphaCurve {keys [<at 0 1 RCIM_Cubic>,<at 1 1 RCIM_Cubic>]}
         BlueCurve {keys [<at 0.132 1 RCIM_Cubic>,<at 0.2775 1>,<at 0.6228 1 RCIM_Cubic>]}
         GreenCurve {keys [<at 0.132 0.6209 RCIM_Cubic>,<at 0.2775 0.0682>,<at 0.6228 0.0015 RCIM_Cubic>]}
         RedCurve {keys [<at 0.132 0.0534 RCIM_Cubic>,<at 0.2775 0.4444>,<at 0.6228 0 RCIM_Cubic>]}
        }})
      Scale Factor=&MakeFloat4FromVectorAndFloat(
       "Vector (RGB)"=&VectorFromFloat(Value=&FloatFromCurve(
         FloatCurve=DataInterface() {propertyValues {Curve {keys [
             {arriveTangent=-0.453, interpMode=RCIM_Cubic, leaveTangent=-0.453, tangentMode=RCTM_User, time=0.5, value=1}
             {arriveTangent=-1.9968, interpMode=RCIM_Cubic, leaveTangent=-1.9968, tangentMode=RCTM_User, time=1, value=0.2}
            ]}}}
         Scale Curve=200
        ))
       "Float (Alpha)"=1
      )
     ))
    ScaleA=true
    ScaleRGB=true
   )
   &UpdateMeshOrientation(Rotation Vector=<0 1 0>) {enabled=false, module=UpdateMeshOrientation001}
   &SkeletalMeshLocation(
    Bone / Socket Index=4
    MaskSpawnGroup=false
    Mesh Sampling Type=::NewEnumerator5
    Orientation Sampling=::NewEnumerator1
    Position Sampling=::NewEnumerator2
    Skeletal Mesh=$User.SkeletalMesh
    Socket Sampling Mode=::NewEnumerator5
   )
   &InheritSourceMovement(
    Applied Force Scale=&Multiply_VectorByFloat(Vector=<0.2 0.2 0.8>, Float=1)
    Applied Position Scale=&Multiply_VectorByFloat(Vector=<-0.03 -0.03 0>, Float=0.1)
    ApplyForce=true
    ApplyForceFalloff=false
    ApplyPosition=false
    Module.LimitForceSpeed=false
   ) {enabled=false}
   &UpdateMeshOrientation(
    Facing Direction=&RotateVectorByQuaternion(Vector To Rotate=<1 0 0>, Quaternion=$Particles.SkeletalMeshLocation.SampledRotation)
    Facing Interpolation=false
    Orientation Method=::NewEnumerator1
    Override Quat=false
    TransformFacing=false
    TransformUp=false
    Up Direction:Vector3f=&RotateVectorByQuaternion(Vector To Rotate=<0 0 1>, Quaternion=$Particles.SkeletalMeshLocation.SampledRotation)
   ) {enabled=false}
   /SetVariables(Particles.LerpAttachedOrientation:NiagaraFloat=&FloatFromCurve(
     FloatCurve=DataInterface() {propertyValues {Curve {keys [{tangentMode=RCTM_User, time=0.3, value=0.5}, {tangentMode=RCTM_User, time=0.4, value=1}]}}}
     CurveIndex=$Emitter.NormalizedLoopAge
    )) {isSetParameters=true, module=SetVariables_#5}
   /SetVariables(Particles.PositionLerped:NiagaraPosition=&LerpPosition(A=$Particles.Position, B=$Particles.SkeletalMeshLocation.SampledPosition, Alpha=$Particles.LerpAttachedOrientation)) {isSetParameters=true, module=SetVariables_#6}
   /SetVariables(
    Particles.MeshOrientationLerped:Quat4f=&LerpQuaternion(
     Quaternion A:Quat4f=$Particles.MeshOrientation
     Quaternion B:Quat4f=&MultiplyQuaternion(
      Quaternion A=$Particles.SkeletalMeshLocation.SampledRotation
      Quaternion B=&MakeQuatFromAxisAngle(Axis=<0 0 1>, AngleInDegrees=90)
     )
     Lerp Factor:NiagaraFloat=&FloatFromCurve(FloatCurve=DataInterface() {propertyValues {Curve {keys [{interpMode=RCIM_Cubic, tangentMode=RCTM_User, time=0.5, value=0}, {interpMode=RCIM_Cubic, tangentMode=RCTM_User, time=1, value=1}]}}})
     Enable Spherical Linear Interpolation:NiagaraBool=false
    )
    Particles.VelocityLerped:Vector3f=&Lerp_Vector(A=$Particles.Velocity, B=$Particles.SkeletalMeshLocation.SampledVelocity, Alpha=$Particles.LerpAttachedOrientation)
   ) {isSetParameters=true, module=SetVariables_#7}
   &SolveForcesAndVelocity(Clamp Velocity=false, Limit Acceleration=false)
   &DynamicMaterialParameters(
    Index 0 Param 2=&FloatFromCurve(
     FloatCurve=DataInterface() {propertyValues {Curve {keys [
         {arriveTangent=0.2665, interpMode=RCIM_Cubic, leaveTangent=0.2665, tangentMode=RCTM_User, time=0, value=1}
         {arriveTangent=-1, interpMode=RCIM_Cubic, leaveTangent=-1, time=0.3, value=0.925}
         {arriveTangent=-1.3379, interpMode=RCIM_Cubic, leaveTangent=-1.3379, tangentMode=RCTM_User, time=1, value=0}
        ]}}}
     Scale Curve=0.4
    )
    Index 1 Param 1=&FloatFromCurve(
     FloatCurve=DataInterface() {propertyValues {Curve {keys [
         {arriveTangent=0.7336, interpMode=RCIM_Cubic, leaveTangent=0.7336, tangentMode=RCTM_User, time=0, value=0}
         {arriveTangent=3.0033, interpMode=RCIM_Cubic, leaveTangent=3.0033, tangentMode=RCTM_User, time=0.4, value=4}
        ]}}}
     Scale Curve=4
    )
    Index 1 Param 2=150
    Index 1 Param 3=&Multiply_Float(A=$Emitter.Age, B=100)
    Index 1 Param 4=&FloatFromCurve(FloatCurve=DataInterface() {propertyValues {Curve {keys [<at 0 1>,<at 1 0>]}}})
    Index 2 Param 1=&FloatFromCurve(
     FloatCurve=DataInterface() {propertyValues {Curve {keys [
         {interpMode=RCIM_Cubic, tangentMode=RCTM_User, time=0, value=0}
         {interpMode=RCIM_Cubic, tangentMode=RCTM_User, time=0.3, value=1}
         {tangentMode=RCTM_User, time=1, value=0.5}
        ]}}}
     Scale Curve=0.5
    )
    Index 2 Param 2=&FloatFromCurve(
     FloatCurve=DataInterface() {propertyValues {Curve {keys [
         {interpMode=RCIM_Cubic, tangentMode=RCTM_User, time=0, value=0}
         {interpMode=RCIM_Cubic, tangentMode=RCTM_User, time=0.3, value=1}
         {tangentMode=RCTM_User, time=1, value=0.5}
        ]}}}
     Scale Curve=2
    )
    Index 3 Param 2=1
    Index 3 Param 4=&FloatFromCurve(
     FloatCurve=DataInterface() {propertyValues {Curve {keys [
         {interpMode=RCIM_Cubic, tangentMode=RCTM_User, time=0, value=1}
         {arriveTangent=-3.5514, interpMode=RCIM_Cubic, leaveTangent=-3.5514, tangentMode=RCTM_User, time=1, value=0}
        ]}}}
     CurveIndex=$System.NormalizedRightDot
    )
    Module.Param0WriteEnabled=true
    Module.Param0WriteEnabled1=true
    Module.Param0WriteEnabled2=true
    Module.Param0WriteEnabled3=true
    Module.Param1WriteEnabled=true
    Module.Param1WriteEnabled1=true
    Module.Param1WriteEnabled2=true
    Module.Param1WriteEnabled3=true
    Module.Param2WriteEnabled=true
    Module.Param2WriteEnabled1=true
    Module.Param2WriteEnabled2=true
    Module.Param2WriteEnabled3=true
    Module.Param3WriteEnabled=true
    Module.Param3WriteEnabled1=true
    Module.Param3WriteEnabled2=true
    Module.Param3WriteEnabled3=true
    Write Parameter Index 1=true
    Write Parameter Index 2=true
    Write Parameter Index 3=true
   )
   /SetVariables(
    Particles.CameraOffset2:NiagaraFloat=&Multiply_Float(A=$Particles.CameraOffset, B=1)
    Particles.Color2=&MultiplyLinearColor(
     Linear Color=&LerpLinearColors(
      StartColor=&ColorFromCurve(ColorCurve=DataInterface() {propertyValues {
         AlphaCurve {keys [<at 0 1 RCIM_Cubic>,<at 1 1 RCIM_Cubic>]}
         BlueCurve {keys [<at 0.1443 1 RCIM_Cubic>,<at 0.4893 1>,<at 0.8109 1 RCIM_Cubic>]}
         GreenCurve {keys [<at 0.1443 0 RCIM_Cubic>,<at 0.4893 0.0652>,<at 0.8109 0.0826 RCIM_Cubic>]}
         RedCurve {keys [<at 0.1443 0.656 RCIM_Cubic>,<at 0.4893 0.3115>,<at 0.8109 0.3508 RCIM_Cubic>]}
        }})
      EndColor=&ColorFromCurve(ColorCurve=DataInterface() {propertyValues {
         AlphaCurve {keys [<at 0 1 RCIM_Cubic>,<at 1 1 RCIM_Cubic>]}
         BlueCurve {keys [<at 0.3066 1 RCIM_Cubic>, {arriveTangent=-1.6736, interpMode=RCIM_Cubic, leaveTangent=-1.6736, time=0.506, value=1}, <at 0.7846 0.2>]}
         GreenCurve {keys [<at 0.3066 0.2436 RCIM_Cubic>, {arriveTangent=-0.5096, interpMode=RCIM_Cubic, leaveTangent=-0.5096, time=0.506, value=0}, <at 0.7846 0>]}
         RedCurve {keys [<at 0.3066 0 RCIM_Cubic>, {arriveTangent=0.1552, interpMode=RCIM_Cubic, leaveTangent=0.1552, time=0.506, value=0.9}, <at 0.7846 0.0742>]}
        }})
      LerpFactor=$User.ForwardDot
     )
     Scale Factor=&MakeFloat4FromVectorAndFloat(
      "Vector (RGB)"=&VectorFromFloat(Value=&FloatFromCurve(
        FloatCurve=DataInterface() {propertyValues {Curve {keys [
            {arriveTangent=0.0734, interpMode=RCIM_Cubic, leaveTangent=0.0734, tangentMode=RCTM_User, time=0, value=1}
            {interpMode=RCIM_Cubic, tangentMode=RCTM_User, time=1, value=1}
           ]}}}
        Scale Curve=20
       ))
      "Float (Alpha)"=&Multiply_Float(
       A=1
       B=&MakeFloatFromLinearColor(LinearColor=$Particles.Color, Channel:ENiagaraLinearColor_Channels=::NewEnumerator3)
      )
     )
    )
    Particles.Position2:NiagaraPosition=&AddVectorToPosition(
     Position=$Particles.Position
     Vector=&MakeVector(
      X=50
      Y=50
      Z=&FloatFromCurve(FloatCurve=DataInterface() {propertyValues {Curve {keys [
           {interpMode=RCIM_Cubic, tangentMode=RCTM_User, time=0, value=-5}
           {interpMode=RCIM_Cubic, tangentMode=RCTM_User, time=0.5, value=5}
           {tangentMode=RCTM_User, time=1, value=-3}
          ]}}})
     )
    )
    Particles.Scale2:Vector3f=&Multiply_Vector(A=$Particles.Scale, B=<1 1 0.8>)
   ) {isSetParameters=true, module=SetVariables_#8}
   &CameraOffset(Camera Offset Amount=-20)
  ]
  rendererClasses [&NiagaraSpriteRendererProperties, &NiagaraMeshRendererProperties]
  renderers [
   {rendererClass=&NiagaraSpriteRendererProperties}
   {rendererClass=&NiagaraMeshRendererProperties, rendererIndex=1}
   {rendererClass=&NiagaraMeshRendererProperties, rendererIndex=2}
   {rendererClass=&NiagaraMeshRendererProperties, rendererIndex=3}
  ]
 }
 Ripples_Deploy001 {
  emitterUpdateScript [
   &EmitterState(Life Cycle Mode=::NewEnumerator1, Loop Behavior=::NewEnumerator1, Loop Delay=0.2, Loop Duration=2.4, UseLoopDelay=true)
   &SpawnRate(
    Spawn Probability=&FN_FrontEndFloatModifier(Gameplay=0, Front End=1)
    SpawnRate=&FloatFromCurve(FloatCurve=DataInterface() {propertyValues {Curve {keys [<at 0.7 1>,<at 1 1>]}}}, CurveIndex=$Emitter.NormalizedLoopAge, Scale Curve=40)
    Use Spawn Probability=true
   )
   &SpawnPerUnit(
    Max Movement Threshold=2000
    Module.Use Max Movement Threshold=true
    Module.Use Movement Tolerance=true
    Movement Tolerance=2
    Spawn Spacing=80
    Use Spawn Probability=false
   ) {enabled=false}
  ]
  particleSpawnScript [
   &InitializeParticle(
    Color=<rgba 1 1 1 0.5>
    Lifetime=&FloatFromCurve(
     FloatCurve=DataInterface() {propertyValues {Curve {keys [<at 0.5 0.2>,<at 0.9 0.55>,<at 1 1>]}}}
     CurveIndex=$Emitter.NormalizedLoopAge
     Scale Curve=2
    )
    Mesh Scale Mode=::NewEnumerator3
    Position Mode=::NewEnumerator0
    Write Material Random=true
   )
   &InitialMeshOrientation(Mesh Axis To Orient=<0 -1 0>, Mesh Orientation Mode=::NewEnumerator2, Use Rotation Vector=false)
   /SetVariables(Particles.AlphaFadeIn:NiagaraFloat=&FloatFromCurve(
     FloatCurve=DataInterface() {propertyValues {Curve {keys [{tangentMode=RCTM_User, time=0.5, value=1}, {tangentMode=RCTM_User, time=1, value=0}, <at 1.2 0>, <at 1.4 1>]}}}
     CurveIndex=$Emitter.Age
    )) {isSetParameters=true, module=SetVariables_#4}
   &SkeletalMeshLocation(
    Bone / Socket Index=4
    MaskSpawnGroup=false
    Mesh Sampling Type=::NewEnumerator5
    Skeletal Mesh=$User.SkeletalMesh
    Socket Sampling Mode=::NewEnumerator5
   ) {module=SkeletalMeshLocation001}
   &InheritVelocity(
    Inherited Velocity Amount Scale=&MakeVector(
     X=0.9
     Y=0.9
     Z=&FloatFromCurve(FloatCurve=DataInterface() {propertyValues {Curve {keys [<at 0 1.1>,<at 1 0.9>]}}})
    )
    Inherited Velocity Speed Limit=10000
    Speed Limit=false
    UseSpeedThreshold=false
   )
   &InheritVelocity(
    Inherited Velocity Amount Scale=&MakeVector(X=1, Y=1, Z=1)
    Inherited Velocity Speed Limit=10000
    Speed Limit=false
    UseSpeedThreshold=false
   ) {enabled=false, module=InheritVelocity001}
   &AddVelocity(Velocity=&MakeVector(
     X=0
     Y=&Add_Float(
      A=&Lerp_Float(A=0, B=500, Alpha=$System.Timeline.Time)
      B=&Lerp_Float(
       A=0
       B=20
       Alpha=&Abs_Float(Float:NiagaraFloat=&Multiply_Float(A=$User.RightDot, B=0.5))
      )
     )
     Z=-2
    )) {enabled=false}
   &AddVelocity(Velocity=<0 1000 0>) {enabled=false, module=AddVelocity001}
   &ApplyInitialForces()
  ]
  particleUpdateScript [
   &ParticleState()
   &ScaleMeshSize(Scale Factor=&VectorFromCurve(
     VectorCurve=DataInterface() {propertyValues {
       XCurve {keys [
         <at 0.2 1.05 RCIM_Cubic>
         {arriveTangent=-1.1469, interpMode=RCIM_Cubic, leaveTangent=-1.1469, tangentMode=RCTM_User, time=1.5, value=0.7}
         <at 2.4 0.1546>
        ]}
       YCurve {keys [
         <at 0.2 1.05 RCIM_Cubic>
         {arriveTangent=-1.1469, interpMode=RCIM_Cubic, leaveTangent=-1.1469, tangentMode=RCTM_User, time=1.5, value=0.7}
         <at 2.4 0.1546>
        ]}
       ZCurve {keys [
         <at 0.2 1.05 RCIM_Cubic>
         {arriveTangent=-1.1469, interpMode=RCIM_Cubic, leaveTangent=-1.1469, tangentMode=RCTM_User, time=1.5, value=0.7}
         <at 2.4 0.1546>
        ]}
      }}
     CurveIndex=$Emitter.Age
    )) {module=ScaleMeshSize004}
   &ScaleMeshSize(Scale Factor=&AddVectorToPosition(
     Position=<1 1 0.5>
     Vector=&VectorFromCurve(
      VectorCurve=DataInterface() {propertyValues {
        XCurve {keys [
          {arriveTangent=-1.2724, interpMode=RCIM_Cubic, leaveTangent=-1.2724, tangentMode=RCTM_User, time=0, value=1}
          {interpMode=RCIM_Cubic, tangentMode=RCTM_User, time=1, value=0}
         ]}
        YCurve {keys [
          {arriveTangent=-1.2724, interpMode=RCIM_Cubic, leaveTangent=-1.2724, tangentMode=RCTM_User, time=0, value=1}
          {interpMode=RCIM_Cubic, tangentMode=RCTM_User, time=1, value=0}
         ]}
        ZCurve {keys [
          {arriveTangent=-1.2724, interpMode=RCIM_Cubic, leaveTangent=-1.2724, tangentMode=RCTM_User, time=0, value=1}
          {interpMode=RCIM_Cubic, tangentMode=RCTM_User, time=1, value=0}
         ]}
       }}
      Scale Curve=<0.5 0.6 0.5>
     )
    )) {enabled=false, module=ScaleMeshSize001}
   &ScaleMeshSize(Scale Factor=&MakeVector(X=&Lerp_Float(A=1, B=1.069, Alpha=$System.Timeline.Time), Y=1, Z=1)) {enabled=false}
   &ScaleMeshSize(Scale Factor=&AddVectorToPosition(
     Position=<1 1 1>
     Vector=&VectorFromCurve(VectorCurve=DataInterface() {propertyValues {
        XCurve {keys [
          {interpMode=RCIM_Cubic, tangentMode=RCTM_User, time=0.2, value=0}
          {arriveTangent=1.25, interpMode=RCIM_Cubic, leaveTangent=1.25, time=0.8, value=0.8}
          {arriveTangent=0.3863, interpMode=RCIM_Cubic, leaveTangent=0.3863, tangentMode=RCTM_User, time=1, value=1}
         ]}
        YCurve {keys [
          {arriveTangent=-0.6221, interpMode=RCIM_Cubic, leaveTangent=-0.6221, tangentMode=RCTM_User, time=0, value=0.4}
          {arriveTangent=1.6, interpMode=RCIM_Cubic, leaveTangent=1.6, time=0.1, value=0.125}
          {arriveTangent=3.8993, interpMode=RCIM_Cubic, leaveTangent=3.8993, tangentMode=RCTM_User, time=1, value=2}
         ]}
        ZCurve {keys [{interpMode=RCIM_Cubic, tangentMode=RCTM_User, time=0, value=1}, {interpMode=RCIM_Cubic, tangentMode=RCTM_User, time=1, value=1}]}
       }})
    )) {enabled=false, module=ScaleMeshSize002}
   &ScaleMeshSize(Scale Factor=&MakeVector(
     X=&Lerp_Float(
      A=&FloatFromCurve(FloatCurve=DataInterface() {propertyValues {Curve {keys [{tangentMode=RCTM_User, time=0, value=1}, {tangentMode=RCTM_User, time=1, value=1.2}]}}})
      B=1
      Alpha=$System.Timeline.Time
     )
     Y=&Lerp_Float(A=0.9, B=0.9, Alpha=$System.Timeline.Time)
     Z=1
    )) {enabled=false, module=ScaleMeshSize003}
   &ScaleColor(
    Scale Alpha=&Multiply_Float(
     A=&FloatFromCurve(FloatCurve=DataInterface() {propertyValues {Curve {keys [
          {arriveTangent=9.0725, interpMode=RCIM_Cubic, leaveTangent=9.0725, tangentMode=RCTM_User, time=0, value=0}
          {arriveTangent=-2.5249, interpMode=RCIM_Cubic, leaveTangent=-2.5249, tangentMode=RCTM_User, time=0.1408, value=0.9847}
          <at 0.2111 0.1382>
          {arriveTangent=-0.0405, interpMode=RCIM_Cubic, leaveTangent=-0.0405, tangentMode=RCTM_User, time=1, value=0}
         ]}}})
     B=$Particles.AlphaFadeIn
    )
    Scale RGB=&MakeVectorFromLinearColorRGB(Color=&MultiplyLinearColor(
      Linear Color=&ColorFromCurve(ColorCurve=DataInterface() {propertyValues {
         AlphaCurve {keys [<at 0 1 RCIM_Cubic>,<at 1 1 RCIM_Cubic>]}
         BlueCurve {keys [<at 0.132 1 RCIM_Cubic>,<at 0.2775 1>,<at 0.6228 1 RCIM_Cubic>]}
         GreenCurve {keys [<at 0.132 0.6209 RCIM_Cubic>,<at 0.2775 0.0682>,<at 0.6228 0.0015 RCIM_Cubic>]}
         RedCurve {keys [<at 0.132 0.0534 RCIM_Cubic>,<at 0.2775 0.4444>,<at 0.6228 0 RCIM_Cubic>]}
        }})
      Scale Factor=&MakeFloat4FromVectorAndFloat(
       "Vector (RGB)"=&VectorFromFloat(Value=&FloatFromCurve(
         FloatCurve=DataInterface() {propertyValues {Curve {keys [
             {arriveTangent=-0.453, interpMode=RCIM_Cubic, leaveTangent=-0.453, tangentMode=RCTM_User, time=0.5, value=1}
             {arriveTangent=-1.9968, interpMode=RCIM_Cubic, leaveTangent=-1.9968, tangentMode=RCTM_User, time=1, value=0.2}
            ]}}}
         Scale Curve=200
        ))
       "Float (Alpha)"=1
      )
     ))
    ScaleA=true
    ScaleRGB=true
   )
   &UpdateMeshOrientation(Rotation Vector=<0 1 0>) {enabled=false, module=UpdateMeshOrientation001}
   &SkeletalMeshLocation(
    Bone / Socket Index=4
    MaskSpawnGroup=false
    Mesh Sampling Type=::NewEnumerator5
    Orientation Sampling=::NewEnumerator1
    Position Sampling=::NewEnumerator2
    Skeletal Mesh=$User.SkeletalMesh
    Socket Sampling Mode=::NewEnumerator5
   )
   &InheritSourceMovement(
    Applied Force Scale=&Multiply_VectorByFloat(Vector=<0.2 0.2 0.8>, Float=1)
    Applied Position Scale=&Multiply_VectorByFloat(Vector=<-0.03 -0.03 0>, Float=0.1)
    ApplyForce=true
    ApplyForceFalloff=false
    ApplyPosition=false
    Module.LimitForceSpeed=false
   ) {enabled=false}
   &UpdateMeshOrientation(
    Facing Direction=&RotateVectorByQuaternion(Vector To Rotate=<1 0 0>, Quaternion=$Particles.SkeletalMeshLocation.SampledRotation)
    Facing Interpolation=false
    Orientation Method=::NewEnumerator1
    Override Quat=false
    TransformFacing=false
    TransformUp=false
    Up Direction:Vector3f=&RotateVectorByQuaternion(Vector To Rotate=<0 0 1>, Quaternion=$Particles.SkeletalMeshLocation.SampledRotation)
   ) {enabled=false}
   /SetVariables(Particles.LerpAttachedOrientation:NiagaraFloat=&FloatFromCurve(
     FloatCurve=DataInterface() {propertyValues {Curve {keys [{tangentMode=RCTM_User, time=0, value=0.5}, {tangentMode=RCTM_User, time=0.1, value=0.7}]}}}
     CurveIndex=$Emitter.NormalizedLoopAge
    )) {isSetParameters=true, module=SetVariables_#5}
   /SetVariables(Particles.PositionLerped:NiagaraPosition=&LerpPosition(A=$Particles.Position, B=$Particles.SkeletalMeshLocation.SampledPosition, Alpha=$Particles.LerpAttachedOrientation)) {isSetParameters=true, module=SetVariables_#6}
   /SetVariables(
    Particles.MeshOrientationLerped:Quat4f=&LerpQuaternion(
     Quaternion A:Quat4f=$Particles.MeshOrientation
     Quaternion B:Quat4f=&MultiplyQuaternion(
      Quaternion A=$Particles.SkeletalMeshLocation.SampledRotation
      Quaternion B=&MakeQuatFromAxisAngle(Axis=<0 0 1>, AngleInDegrees=90)
     )
     Lerp Factor:NiagaraFloat=&FloatFromCurve(FloatCurve=DataInterface() {propertyValues {Curve {keys [{interpMode=RCIM_Cubic, tangentMode=RCTM_User, time=0.5, value=0}, {interpMode=RCIM_Cubic, tangentMode=RCTM_User, time=1, value=1}]}}})
     Enable Spherical Linear Interpolation:NiagaraBool=false
    )
    Particles.VelocityLerped:Vector3f=&Lerp_Vector(A=$Particles.Velocity, B=$Particles.SkeletalMeshLocation.SampledVelocity, Alpha=$Particles.LerpAttachedOrientation)
   ) {isSetParameters=true, module=SetVariables_#7}
   &SolveForcesAndVelocity(Clamp Velocity=false, Limit Acceleration=false)
   &DynamicMaterialParameters(
    Index 0 Param 2=&FloatFromCurve(
     FloatCurve=DataInterface() {propertyValues {Curve {keys [
         {arriveTangent=0.2665, interpMode=RCIM_Cubic, leaveTangent=0.2665, tangentMode=RCTM_User, time=0, value=1}
         {arriveTangent=-1, interpMode=RCIM_Cubic, leaveTangent=-1, time=0.3, value=0.925}
         {arriveTangent=-1.3379, interpMode=RCIM_Cubic, leaveTangent=-1.3379, tangentMode=RCTM_User, time=1, value=0}
        ]}}}
     Scale Curve=0.4
    )
    Index 1 Param 1=&FloatFromCurve(
     FloatCurve=DataInterface() {propertyValues {Curve {keys [
         {arriveTangent=0.7336, interpMode=RCIM_Cubic, leaveTangent=0.7336, tangentMode=RCTM_User, time=0, value=0}
         {arriveTangent=3.0033, interpMode=RCIM_Cubic, leaveTangent=3.0033, tangentMode=RCTM_User, time=0.4, value=4}
        ]}}}
     Scale Curve=4
    )
    Index 1 Param 2=150
    Index 1 Param 3=&Multiply_Float(A=$Emitter.Age, B=100)
    Index 1 Param 4=&FloatFromCurve(FloatCurve=DataInterface() {propertyValues {Curve {keys [<at 0 1>,<at 1 0>]}}})
    Index 2 Param 1=&FloatFromCurve(
     FloatCurve=DataInterface() {propertyValues {Curve {keys [
         {interpMode=RCIM_Cubic, tangentMode=RCTM_User, time=0, value=0}
         {interpMode=RCIM_Cubic, tangentMode=RCTM_User, time=0.3, value=1}
         {tangentMode=RCTM_User, time=1, value=0.5}
        ]}}}
     Scale Curve=0.5
    )
    Index 2 Param 2=&FloatFromCurve(
     FloatCurve=DataInterface() {propertyValues {Curve {keys [
         {interpMode=RCIM_Cubic, tangentMode=RCTM_User, time=0, value=0}
         {interpMode=RCIM_Cubic, tangentMode=RCTM_User, time=0.3, value=1}
         {tangentMode=RCTM_User, time=1, value=0.5}
        ]}}}
     Scale Curve=2
    )
    Index 3 Param 2=1
    Index 3 Param 4=&FloatFromCurve(
     FloatCurve=DataInterface() {propertyValues {Curve {keys [
         {interpMode=RCIM_Cubic, tangentMode=RCTM_User, time=0, value=1}
         {arriveTangent=-3.5514, interpMode=RCIM_Cubic, leaveTangent=-3.5514, tangentMode=RCTM_User, time=1, value=0}
        ]}}}
     CurveIndex=$System.NormalizedRightDot
    )
    Module.Param0WriteEnabled=true
    Module.Param0WriteEnabled1=true
    Module.Param0WriteEnabled2=true
    Module.Param0WriteEnabled3=true
    Module.Param1WriteEnabled=true
    Module.Param1WriteEnabled1=true
    Module.Param1WriteEnabled2=true
    Module.Param1WriteEnabled3=true
    Module.Param2WriteEnabled=true
    Module.Param2WriteEnabled1=true
    Module.Param2WriteEnabled2=true
    Module.Param2WriteEnabled3=true
    Module.Param3WriteEnabled=true
    Module.Param3WriteEnabled1=true
    Module.Param3WriteEnabled2=true
    Module.Param3WriteEnabled3=true
    Write Parameter Index 1=true
    Write Parameter Index 2=true
    Write Parameter Index 3=true
   )
   /SetVariables(
    Particles.CameraOffset2:NiagaraFloat=&Multiply_Float(A=$Particles.CameraOffset, B=1)
    Particles.Color2=&MultiplyLinearColor(
     Linear Color=&LerpLinearColors(
      StartColor=&ColorFromCurve(ColorCurve=DataInterface() {propertyValues {
         AlphaCurve {keys [<at 0 1 RCIM_Cubic>,<at 1 1 RCIM_Cubic>]}
         BlueCurve {keys [<at 0.1443 1 RCIM_Cubic>,<at 0.4893 1>,<at 0.8109 1 RCIM_Cubic>]}
         GreenCurve {keys [<at 0.1443 0 RCIM_Cubic>,<at 0.4893 0.0652>,<at 0.8109 0.0826 RCIM_Cubic>]}
         RedCurve {keys [<at 0.1443 0.656 RCIM_Cubic>,<at 0.4893 0.3115>,<at 0.8109 0.3508 RCIM_Cubic>]}
        }})
      EndColor=&ColorFromCurve(ColorCurve=DataInterface() {propertyValues {
         AlphaCurve {keys [<at 0 1 RCIM_Cubic>,<at 1 1 RCIM_Cubic>]}
         BlueCurve {keys [<at 0.3066 1 RCIM_Cubic>, {arriveTangent=-1.6736, interpMode=RCIM_Cubic, leaveTangent=-1.6736, time=0.506, value=1}, <at 0.7846 0.2>]}
         GreenCurve {keys [<at 0.3066 0.2436 RCIM_Cubic>, {arriveTangent=-0.5096, interpMode=RCIM_Cubic, leaveTangent=-0.5096, time=0.506, value=0}, <at 0.7846 0>]}
         RedCurve {keys [<at 0.3066 0 RCIM_Cubic>, {arriveTangent=0.1552, interpMode=RCIM_Cubic, leaveTangent=0.1552, time=0.506, value=0.9}, <at 0.7846 0.0742>]}
        }})
      LerpFactor=$User.ForwardDot
     )
     Scale Factor=&MakeFloat4FromVectorAndFloat(
      "Vector (RGB)"=&VectorFromFloat(Value=&FloatFromCurve(FloatCurve=DataInterface() {propertyValues {Curve {keys [
            {arriveTangent=0.0734, interpMode=RCIM_Cubic, leaveTangent=0.0734, tangentMode=RCTM_User, time=0, value=1}
            {interpMode=RCIM_Cubic, tangentMode=RCTM_User, time=1, value=1}
           ]}}}))
      "Float (Alpha)"=&Multiply_Float(
       A=30
       B=&MakeFloatFromLinearColor(LinearColor=$Particles.Color, Channel:ENiagaraLinearColor_Channels=::NewEnumerator3)
      )
     )
    )
    Particles.Position2:NiagaraPosition=&AddVectorToPosition(
     Position=$Particles.Position
     Vector=&MakeVector(
      X=50
      Y=50
      Z=&FloatFromCurve(FloatCurve=DataInterface() {propertyValues {Curve {keys [
           {interpMode=RCIM_Cubic, tangentMode=RCTM_User, time=0, value=-5}
           {interpMode=RCIM_Cubic, tangentMode=RCTM_User, time=0.5, value=5}
           {tangentMode=RCTM_User, time=1, value=-3}
          ]}}})
     )
    )
    Particles.Scale2:Vector3f=&Multiply_Vector(A=$Particles.Scale, B=<1 1 0.8>)
   ) {isSetParameters=true, module=SetVariables_#8}
   &CameraOffset(Camera Offset Amount=-20)
  ]
  rendererClasses [&NiagaraSpriteRendererProperties, &NiagaraMeshRendererProperties]
  renderers [
   {rendererClass=&NiagaraSpriteRendererProperties}
   {rendererClass=&NiagaraMeshRendererProperties, rendererIndex=1}
   {rendererClass=&NiagaraMeshRendererProperties, rendererIndex=2}
   {rendererClass=&NiagaraMeshRendererProperties, rendererIndex=3}
  ]
 }
 Speedlines {
  emitterUpdateScript [
   &EmitterState(Enable Distance Culling=true, MaxDistance=2048, Module.bUseMaxDistance=true, Module.bUseMinDistance=false, Scalability Mode=::NewEnumerator1)
   &SpawnRate(
    SpawnRate=&Multiply_Float(
     A=&FN_FrontEndFloatModifier(Gameplay=$User.ForwardDot, Front End=1)
     B=&UniformRangedFloat(Minimum=75, Maximum=150)
    )
    Use Spawn Probability=false
   )
  ]
  particleSpawnScript [
   /Niagara/Modules/Spawn/Initialization/InitializeParticle(
    Color=&MakeLinearColorFromVectorAndFloat(
     "Vector (RGB)"=<1 1 1>
     "Float (Alpha)"=&Multiply_Float(
      A=&UniformRangedFloat(Minimum=0.015, Maximum=0.05)
      B=&Add_Float(
       A=1
       B=&Multiply_Float(A=10, B=&Subtract_Float(A=1, B=$Engine.Emitter.SpawnCountScale))
      )
     )
    )
    Lifetime=&FN_FrontEndFloatModifier(
     Gameplay=&UniformRangedFloat(Minimum=0.2, Maximum=0.35)
     Front End=&UniformRangedFloat(Minimum=0.1, Maximum=0.2)
    )
    Module.Write Color=true
    Module.Write Lifetime=true
    Module.Write Mass=true
    Module.Write Material Random=true
    Module.Write Position=false
    Module.Write Scale=false
    Module.Write SpriteRotation=false
    Module.Write SpriteSize=true
    Position=&SimulationPosition()
    Sprite Size=&UniformRangedVector2D(Minimum=<3 4>, Maximum=<5 6>)
   )
   &CylinderLocation(
    Cylinder Radius=&UniformRangedFloat(Minimum=100, Maximum=150)
    MaskSpawnGroup=false
    Module.Enable Lathe Profile=false
    Module.Override Randomness=false
    Module.Surface Only=true
    Orientation Axis:ENiagaraOrientationAxis=::XAxis
    Use Endcaps In Surface Only Mode=false
   )
   &AddVelocity(Velocity=&FN_FrontEndVector3Modifier(Gameplay=<0 1000 0>, Front End=<0 3000 0>))
   &AddVelocityFromPoint(
    Module.Use Falloff Distance=false
    Velocity Strength=&Multiply_Float(
     A=&FN_FrontEndFloatModifier(Gameplay=1, Front End=2)
     B=&UniformRangedFloat(Minimum=500, Maximum=700)
    )
   )
  ]
  particleUpdateScript [
   &ParticleState()
   &SolveForcesAndVelocity(Clamp Velocity=false, Limit Acceleration=false)
   &ScaleSpriteSize(Scale Factor=&Vector2DFromCurve(
     Vector2Curve=DataInterface() {propertyValues {XCurve {keys [<at 0 0.5 RCIM_Cubic>,<at 0.5 1 RCIM_Cubic>]}, YCurve {keys [<at 0 0 RCIM_Cubic>,<at 0.5 1 RCIM_Cubic>]}}}
     ScaleCurve=&Vector2DFromFloat(Value=1)
    ))
   &ScaleSpriteSizeBySpeed(Max Scale Factor=<1 40>, Min Scale Factor=<0.01 0.05>, Velocity Threshold=1)
   &ScaleColor(
    Scale Alpha=&FloatFromCurve(FloatCurve=DataInterface() {propertyValues {Curve {keys [<at 0 0 RCIM_Cubic>, <at 0.2 1>, {arriveTangent=-2.221, interpMode=RCIM_Cubic, leaveTangent=-2.221, tangentMode=RCTM_User, time=1, value=0}]}}})
    ScaleA=true
    ScaleRGB=true
   )
  ]
  rendererClasses [&NiagaraSpriteRendererProperties]
  renderers [{rendererClass=&NiagaraSpriteRendererProperties}]
 }
 Splash_Drops {
  emitterUpdateScript [
   &EmitterState(Inactive Response=::NewEnumerator1, Life Cycle Mode=::NewEnumerator1, Loop Behavior=::NewEnumerator1, Loop Duration=3600, UseLoopDelay=false)
   &SpawnRate(
    Spawn Probability=&FN_FrontEndFloatModifier(Gameplay=1, Front End=0)
    SpawnRate=&FloatFromCurve(FloatCurve=DataInterface() {propertyValues {Curve {keys [<at 0 1>,<at 1 0>]}}}, CurveIndex=$Emitter.Age, Scale Curve=50)
    Use Spawn Probability=true
   ) {module=SpawnRate001}
   &SpawnRate(
    Spawn Probability=&FN_FrontEndFloatModifier(Gameplay=1, Front End=0)
    SpawnRate=&Add_Float(
     A=&Lerp_Float(A=0, B=50, Alpha=&ClampFloat(Float=$User.ForwardDot))
     B=&Lerp_Float(A=0, B=80, Alpha=&ClampFloat(Float=$System.RightDot_TurnStrength))
    )
    Use Spawn Probability=true
   )
  ]
  particleSpawnScript [
   &FN_SetSpawnGroup(Number of Spawn Groups=2)
   &PartitionParticles()
   /SetVariables(Particles.MyCustomIndex:NiagaraInt32=&SelectIntFromArray(Direct Array Index=$Particles.PartitionParticles.Partition, Int Selection Array=DataInterface() {propertyValues {IntData [2,3]}})) {isSetParameters=true, module=SetVariables_#9}
   /SetVariables(
    Particles.FrontToBack:NiagaraFloat=&RandomRangeFloat(Minimum=0, Maximum=1)
    Particles.LeanLerp:NiagaraFloat=&Multiply_Float(
     A=$User.RightDot
     B=&Lerp_Float(A=-1, B=1, Alpha=&MakeFloatFromInt(Int=$Particles.SpawnGroup))
    )
   ) {isSetParameters=true, module=SetVariables_#10}
   /Niagara/Modules/Spawn/Initialization/V2/InitializeParticle(
    Color=&MakeLinearColorFromVectorAndFloat("Vector (RGB)"=&VectorFromFloat(Value=1), "Float (Alpha)"=1)
    Lifetime Max=0.6
    Lifetime Min=0.3
    Lifetime Mode=::NewEnumerator1
    Sprite Rotation Mode=::NewEnumerator2
    Sprite Size Mode=::NewEnumerator1
    Uniform Sprite Size Max=4
    Uniform Sprite Size Min=2
    UsePositionOffset=false
    Write Material Random=true
   ) {module=InitializeParticle001}
   &SkeletalMeshLocation(
    Apply Sampled Orientation to Mesh Particles=true
    Bone / Socket Index=$Particles.MyCustomIndex
    Inherit Velocity=false
    MaskSpawnGroup=false
    Mesh Sampling Type=::NewEnumerator5
    Sampled Position Offset=&VectorFromCurve(
     VectorCurve=DataInterface() {propertyValues {
       XCurve {keys [<at 0 35>, {arriveTangent=-382.5781, interpMode=RCIM_Cubic, leaveTangent=-382.5781, tangentMode=RCTM_User, time=1, value=-120}]}
       YCurve {keys [<at 0.33 0 RCIM_Cubic>, {arriveTangent=-106.8041, interpMode=RCIM_Cubic, leaveTangent=-106.8041, tangentMode=RCTM_User, time=1, value=-30}]}
       ZCurve {keys [<at 0 5>,<at 1 10>]}
      }}
     CurveIndex=$Particles.FrontToBack
     Scale Curve=&Lerp_Vector(A=<1 1 1>, B=<1 -1 1>, Alpha=&MakeFloatFromInt(Int=$Particles.SpawnGroup))
    )
    Sampled Position Offset Coordinate Space:ENiagaraMeshTransforms=::NewEnumerator3
    Skeletal Mesh=$User.SkeletalMesh
    Socket Sampling Mode=::NewEnumerator5
   )
   /Niagara/Modules/Spawn/Location/SphereLocation(MaskSpawnGroup=false, Module.Override Randomness=false, Sphere Radius=3, Surface Only=false)
   /Niagara/Modules/Update/SubUV/SubUVAnimation(SubUV Animation Mode=::NewEnumerator1)
   /SetVariables(Particles.Velocity=&Multiply_VectorByFloat(
     Vector=&ReturnMeshOrientationAxis(Mesh Orientation Axis=::XAxis, Mesh Orientation Quat=$Particles.MeshOrientation)
     Float=&Lerp_Float(A=-50, B=-100, Alpha=$User.ForwardDot)
    )) {isSetParameters=true, module=SetVariables_#11}
   &AddVelocity(
    Scale Added Velocity=&Lerp_Vector(
     A=<1 1 1>
     B=<1.1 1.1 1.25>
     Alpha=&ClampFloat(
      Float=&Add_Float(A=&ClampFloat(Float=$User.ForwardDot), B=$Particles.LeanLerp)
      Min=-0.25
     )
    )
    Velocity=&RandomRangeVector(Minimum=<-40 120 300>, Maximum=<40 200 200>)
   )
   &AddVelocityFromPoint(Module.Use Falloff Distance=false, Origin Offset=<0 -20 -15>, Origin Offset Coordinate Space=::Simulation)
  ]
  particleUpdateScript [
   &ParticleState()
   &InheritVelocity(
    Module.Limit Speed=true
    Velocity Limit=10000
    Velocity Scale=&Multiply_VectorByFloat(Vector=<-0.03 -0.03 -0.1>, Float=0.1)
    Velocity Source=&TransformVector(Vector=$Engine.Owner.Velocity, Source Space=::World, Destination Space=::Local)
   )
   &SpriteRotationRate(Rotation Rate=&RandomRangeFloat(Minimum=-180, Maximum=180))
   &Drag(Drag=0.5, Use Linear Drag=true, Use Rotational Drag=false)
   &GravityForce(Gravity=<0 0 -900>)
   &SolveForcesAndVelocity(Clamp Velocity=false, Limit Acceleration=false)
   &ScaleSpriteSize(Scale Factor=&Vector2DFromFloat(Value=&FloatFromCurve(
      FloatCurve=DataInterface() {propertyValues {Curve {keys [
          {interpMode=RCIM_Cubic, tangentMode=RCTM_User, time=0.8, value=1}
          {arriveTangent=-3.3474, interpMode=RCIM_Cubic, leaveTangent=-3.3474, tangentMode=RCTM_User, time=1, value=0}
         ]}}}
      Scale Curve=0.65
     )))
   &ScaleColor(
    Scale Alpha=&Multiply_Float(
     A=&FloatFromCurve(FloatCurve=DataInterface() {propertyValues {Curve {keys [
          {arriveTangent=9.0725, interpMode=RCIM_Cubic, leaveTangent=9.0725, tangentMode=RCTM_User, time=0, value=0}
          {arriveTangent=-2.5249, interpMode=RCIM_Cubic, leaveTangent=-2.5249, tangentMode=RCTM_User, time=0.1408, value=0.9847}
          {arriveTangent=-0.0405, interpMode=RCIM_Cubic, leaveTangent=-0.0405, tangentMode=RCTM_User, time=1, value=0}
         ]}}})
     B=1
    )
    Scale RGB=&MakeVectorFromLinearColorRGB(Color=&MultiplyLinearColor(
      Linear Color=&ColorFromCurve(ColorCurve=DataInterface() {propertyValues {
         AlphaCurve {keys [<at 0 1 RCIM_Cubic>,<at 1 1 RCIM_Cubic>]}
         BlueCurve {keys [<at 0.132 1 RCIM_Cubic>,<at 0.4893 1>,<at 0.8109 1 RCIM_Cubic>]}
         GreenCurve {keys [<at 0.132 0.6209 RCIM_Cubic>,<at 0.4893 0.0499>,<at 0.8109 0.0015 RCIM_Cubic>]}
         RedCurve {keys [<at 0.132 0.0534 RCIM_Cubic>,<at 0.4893 0.7665>,<at 0.8109 0 RCIM_Cubic>]}
        }})
      Scale Factor=&MakeFloat4FromVectorAndFloat(
       "Vector (RGB)"=&VectorFromFloat(Value=&FloatFromCurve(
         FloatCurve=DataInterface() {propertyValues {Curve {keys [
             {arriveTangent=-0.453, interpMode=RCIM_Cubic, leaveTangent=-0.453, tangentMode=RCTM_User, time=0.5, value=1}
             {interpMode=RCIM_Cubic, tangentMode=RCTM_User, time=1, value=0.2}
            ]}}}
         Scale Curve=20
        ))
       "Float (Alpha)"=1
      )
     ))
    ScaleA=true
    ScaleRGB=true
   )
   &DynamicMaterialParameters(
    Index 0 Param 1=&RandomRangeFloat(Minimum=0, Maximum=1)
    Index 0 Param 3=&FloatFromCurve(FloatCurve=DataInterface() {propertyValues {Curve {keys [{interpMode=RCIM_Cubic, tangentMode=RCTM_User, time=0, value=0}, {interpMode=RCIM_Cubic, tangentMode=RCTM_User, time=1, value=1}]}}})
    Module.Param0WriteEnabled=true
    Module.Param1WriteEnabled=false
    Module.Param2WriteEnabled=true
    Module.Param3WriteEnabled=true
   )
  ]
  rendererClasses [&NiagaraSpriteRendererProperties]
  renderers [{rendererClass=&NiagaraSpriteRendererProperties}]
 }
 Splash_Drops001 {
  emitterUpdateScript [
   &EmitterState(Inactive Response=::NewEnumerator1, Life Cycle Mode=::NewEnumerator1, Loop Duration=1.3, UseLoopDelay=false)
   &SpawnRate(
    Spawn Probability=&FN_FrontEndFloatModifier(Gameplay=0, Front End=1)
    SpawnRate=&FloatFromCurve(FloatCurve=DataInterface() {propertyValues {Curve {keys [<at 0 1>,<at 0.8 0>]}}}, CurveIndex=$Emitter.LoopedAge, Scale Curve=80)
    Use Spawn Probability=true
   ) {module=SpawnRate001}
   &SpawnRate(
    Spawn Probability=&FN_FrontEndFloatModifier(Gameplay=0, Front End=1)
    SpawnRate=10
    Use Spawn Probability=true
   )
  ]
  particleSpawnScript [
   &FN_SetSpawnGroup(Number of Spawn Groups=2)
   &PartitionParticles()
   /SetVariables(Particles.MyCustomIndex:NiagaraInt32=&SelectIntFromArray(Direct Array Index=$Particles.PartitionParticles.Partition, Int Selection Array=DataInterface() {propertyValues {IntData [2,3]}})) {isSetParameters=true, module=SetVariables_#9}
   /SetVariables(
    Particles.FrontToBack:NiagaraFloat=&RandomRangeFloat(Minimum=0, Maximum=1)
    Particles.LeanLerp:NiagaraFloat=&Multiply_Float(
     A=$User.RightDot
     B=&Lerp_Float(A=-1, B=1, Alpha=&MakeFloatFromInt(Int=$Particles.SpawnGroup))
    )
   ) {isSetParameters=true, module=SetVariables_#10}
   &InitializeParticle(
    Color=&MakeLinearColorFromVectorAndFloat("Vector (RGB)"=&VectorFromFloat(Value=1), "Float (Alpha)"=1)
    Lifetime Max=0.6
    Lifetime Min=0.3
    Lifetime Mode=::NewEnumerator1
    Sprite Rotation Mode=::NewEnumerator2
    Sprite Size Mode=::NewEnumerator1
    Uniform Sprite Size Max=4
    Uniform Sprite Size Min=2
    UsePositionOffset=false
    Write Material Random=true
   ) {module=InitializeParticle001}
   &SkeletalMeshLocation(
    Apply Sampled Orientation to Mesh Particles=true
    Bone / Socket Index=$Particles.MyCustomIndex
    Inherit Velocity=false
    MaskSpawnGroup=false
    Mesh Sampling Type=::NewEnumerator5
    Sampled Position Offset=&VectorFromCurve(
     VectorCurve=DataInterface() {propertyValues {
       XCurve {keys [<at 0 35>, {arriveTangent=-382.5781, interpMode=RCIM_Cubic, leaveTangent=-382.5781, tangentMode=RCTM_User, time=1, value=-120}]}
       YCurve {keys [<at 0.33 0 RCIM_Cubic>, {arriveTangent=-106.8041, interpMode=RCIM_Cubic, leaveTangent=-106.8041, tangentMode=RCTM_User, time=1, value=-30}]}
       ZCurve {keys [<at 0 5>,<at 1 10>]}
      }}
     CurveIndex=$Particles.FrontToBack
     Scale Curve=&Lerp_Vector(A=<1 1 1>, B=<1 -1 1>, Alpha=&MakeFloatFromInt(Int=$Particles.SpawnGroup))
    )
    Sampled Position Offset Coordinate Space:ENiagaraMeshTransforms=::NewEnumerator3
    Skeletal Mesh=$User.SkeletalMesh
    Socket Sampling Mode=::NewEnumerator5
   )
   &SphereLocation(MaskSpawnGroup=false, Module.Override Randomness=false, Sphere Radius=3, Surface Only=false)
   &SubUVAnimation(SubUV Animation Mode=::NewEnumerator1)
   /SetVariables(Particles.Velocity=&Multiply_VectorByFloat(
     Vector=&ReturnMeshOrientationAxis(Mesh Orientation Axis=::XAxis, Mesh Orientation Quat=$Particles.MeshOrientation)
     Float=&Lerp_Float(A=-50, B=-100, Alpha=$User.ForwardDot)
    )) {isSetParameters=true, module=SetVariables_#11}
   &AddVelocity(
    Scale Added Velocity=&Lerp_Vector(
     A=<1 1 1>
     B=<1.1 1.1 1.25>
     Alpha=&ClampFloat(
      Float=&Add_Float(A=&ClampFloat(Float=$User.ForwardDot), B=$Particles.LeanLerp)
      Min=-0.25
     )
    )
    Velocity=&RandomRangeVector(Minimum=<-40 200 300>, Maximum=<40 200 200>)
   )
   &AddVelocityFromPoint(Module.Use Falloff Distance=false, Origin Offset=<0 -20 -15>, Origin Offset Coordinate Space=::Simulation)
  ]
  particleUpdateScript [
   &ParticleState()
   &InheritVelocity(
    Module.Limit Speed=true
    Velocity Limit=10000
    Velocity Scale=&Multiply_VectorByFloat(Vector=<-0.03 -0.03 -0.1>, Float=0.1)
    Velocity Source=&TransformVector(Vector=$Engine.Owner.Velocity, Source Space=::World, Destination Space=::Local)
   )
   &SpriteRotationRate(Rotation Rate=&RandomRangeFloat(Minimum=-180, Maximum=180))
   &Drag(Drag=0.5, Use Linear Drag=true, Use Rotational Drag=false)
   &GravityForce(Gravity=<0 0 -900>)
   &SolveForcesAndVelocity(Clamp Velocity=false, Limit Acceleration=false)
   &ScaleSpriteSize(Scale Factor=&Vector2DFromFloat(Value=&FloatFromCurve(
      FloatCurve=DataInterface() {propertyValues {Curve {keys [
          {interpMode=RCIM_Cubic, tangentMode=RCTM_User, time=0.8, value=1}
          {arriveTangent=-3.3474, interpMode=RCIM_Cubic, leaveTangent=-3.3474, tangentMode=RCTM_User, time=1, value=0}
         ]}}}
      Scale Curve=0.65
     )))
   &ScaleColor(
    Scale Alpha=&Multiply_Float(
     A=&FloatFromCurve(FloatCurve=DataInterface() {propertyValues {Curve {keys [
          {arriveTangent=9.0725, interpMode=RCIM_Cubic, leaveTangent=9.0725, tangentMode=RCTM_User, time=0.08, value=0}
          {arriveTangent=-2.5249, interpMode=RCIM_Cubic, leaveTangent=-2.5249, tangentMode=RCTM_User, time=0.1408, value=0.9847}
          {arriveTangent=-0.0405, interpMode=RCIM_Cubic, leaveTangent=-0.0405, tangentMode=RCTM_User, time=1, value=0}
         ]}}})
     B=1
    )
    Scale RGB=&MakeVectorFromLinearColorRGB(Color=&MultiplyLinearColor(
      Linear Color=&ColorFromCurve(ColorCurve=DataInterface() {propertyValues {
         AlphaCurve {keys [<at 0 1 RCIM_Cubic>,<at 1 1 RCIM_Cubic>]}
         BlueCurve {keys [<at 0.132 1 RCIM_Cubic>,<at 0.4893 1>,<at 0.8109 1 RCIM_Cubic>]}
         GreenCurve {keys [<at 0.132 0.6209 RCIM_Cubic>,<at 0.4893 0.0499>,<at 0.8109 0.0015 RCIM_Cubic>]}
         RedCurve {keys [<at 0.132 0.0534 RCIM_Cubic>,<at 0.4893 0.7665>,<at 0.8109 0 RCIM_Cubic>]}
        }})
      Scale Factor=&MakeFloat4FromVectorAndFloat(
       "Vector (RGB)"=&VectorFromFloat(Value=&FloatFromCurve(
         FloatCurve=DataInterface() {propertyValues {Curve {keys [
             {arriveTangent=-0.453, interpMode=RCIM_Cubic, leaveTangent=-0.453, tangentMode=RCTM_User, time=0.5, value=1}
             {interpMode=RCIM_Cubic, tangentMode=RCTM_User, time=1, value=0.2}
            ]}}}
         Scale Curve=20
        ))
       "Float (Alpha)"=1
      )
     ))
    ScaleA=true
    ScaleRGB=true
   )
   &DynamicMaterialParameters(
    Index 0 Param 1=&RandomRangeFloat(Minimum=0, Maximum=1)
    Index 0 Param 3=&FloatFromCurve(FloatCurve=DataInterface() {propertyValues {Curve {keys [{interpMode=RCIM_Cubic, tangentMode=RCTM_User, time=0, value=0}, {interpMode=RCIM_Cubic, tangentMode=RCTM_User, time=1, value=1}]}}})
    Module.Param0WriteEnabled=true
    Module.Param1WriteEnabled=false
    Module.Param2WriteEnabled=true
    Module.Param3WriteEnabled=true
   )
  ]
  rendererClasses [&NiagaraSpriteRendererProperties]
  renderers [{rendererClass=&NiagaraSpriteRendererProperties}]
 }
 Splash_Foam {
  emitterUpdateScript [
   &EmitterState(Inactive Response=::NewEnumerator1, Life Cycle Mode=::NewEnumerator1, Loop Behavior=::NewEnumerator1, Loop Duration=3600, UseLoopDelay=false)
   &SpawnRate(
    Spawn Probability=&FN_FrontEndFloatModifier(Gameplay=1, Front End=0)
    SpawnRate=&FloatFromCurve(FloatCurve=DataInterface() {propertyValues {Curve {keys [<at 0 1>,<at 1 0>]}}}, CurveIndex=$Emitter.Age, Scale Curve=150)
    Use Spawn Probability=true
   ) {module=SpawnRate001}
   &SpawnRate(
    Spawn Probability=&FN_FrontEndFloatModifier(Gameplay=1, Front End=0)
    SpawnRate=&Add_Float(
     A=&Lerp_Float(A=0, B=30, Alpha=&ClampFloat(Float=$User.ForwardDot))
     B=&Lerp_Float(A=0, B=50, Alpha=&ClampFloat(Float=$System.RightDot_TurnStrength))
    )
    Use Spawn Probability=true
   )
  ]
  particleSpawnScript [
   &FN_SetSpawnGroup(Number of Spawn Groups=2) {enabled=false}
   &PartitionParticles()
   /SetVariables(Particles.MyCustomIndex:NiagaraInt32=&SelectIntFromArray(Direct Array Index=$Particles.PartitionParticles.Partition, Int Selection Array=DataInterface() {propertyValues {IntData [2,3]}})) {isSetParameters=true, module=SetVariables_#12}
   /SetVariables(
    Particles.FrontToBack:NiagaraFloat=&RandomRangeFloat(Minimum=0, Maximum=1)
    Particles.LeanLerp:NiagaraFloat=&Multiply_Float(
     A=$User.RightDot
     B=&Lerp_Float(A=-1, B=1, Alpha=&MakeFloatFromInt(Int=$Particles.MyCustomIndex))
    )
   ) {isSetParameters=true, module=SetVariables_#13}
   &InitializeParticle(
    Lifetime Max=0.4
    Lifetime Min=0.2
    Lifetime Mode=::NewEnumerator1
    Mesh Scale Mode=::NewEnumerator3
    Sprite Rotation Mode=::NewEnumerator2
    Sprite Size Mode=::NewEnumerator1
    Uniform Sprite Size Max=100
    Uniform Sprite Size Min=75
    UsePositionOffset=false
    Write Material Random=true
   ) {module=InitializeParticle001}
   &SkeletalMeshLocation(
    Apply Sampled Orientation to Mesh Particles=true
    Bone / Socket Index=$Particles.MyCustomIndex
    Inherit Velocity=false
    Interpolate Mesh Sampling=false
    MaskSpawnGroup=false
    Mesh Sampling Type=::NewEnumerator5
    Sampled Position Offset=&VectorFromCurve(
     VectorCurve=DataInterface() {propertyValues {
       XCurve {keys [<at 0 20>, {arriveTangent=-382.5781, interpMode=RCIM_Cubic, leaveTangent=-382.5781, tangentMode=RCTM_User, time=1, value=-120}]}
       YCurve {keys [<at 0.33 0 RCIM_Cubic>, {arriveTangent=-92.7914, interpMode=RCIM_Cubic, leaveTangent=-92.7914, tangentMode=RCTM_User, time=1, value=-35}]}
       ZCurve {keys [<at 0 5>,<at 1 10>]}
      }}
     CurveIndex=$Particles.FrontToBack
     Scale Curve=&Lerp_Vector(A=<-1 0.1 1>, B=<-1 -0.1 1>, Alpha=&MakeFloatFromInt(Int=$Particles.MyCustomIndex))
    )
    Sampled Position Offset Coordinate Space:ENiagaraMeshTransforms=::NewEnumerator3
    Skeletal Mesh=$User.SkeletalMesh
    Socket Sampling Mode=::NewEnumerator5
   ) {module=SkeletalMeshLocation001}
   &SubUVAnimation(SubUV Animation Mode=::NewEnumerator1)
   /SetVariables(Particles.Velocity=&Multiply_VectorByFloat(
     Vector=&ReturnMeshOrientationAxis(Mesh Orientation Axis=::XAxis, Mesh Orientation Quat=$Particles.MeshOrientation)
     Float=&Lerp_Float(A=-400, B=-500, Alpha=$User.ForwardDot)
    )) {isSetParameters=true, module=SetVariables_#14}
   &AddVelocity(
    Scale Added Velocity=&Lerp_Vector(
     A=<1 1 1>
     B=<1.1 1.1 1.25>
     Alpha=&Add_Float(A=&ClampFloat(Float=$User.ForwardDot), B=$Particles.LeanLerp)
    )
    Velocity=&RandomRangeVector(Minimum=<-10 -10 175>, Maximum=<10 10 250>)
   )
   &AddVelocityFromPoint(Module.Use Falloff Distance=false, Origin Offset=<0 -20 -15>, Origin Offset Coordinate Space=::Simulation)
  ]
  particleUpdateScript [
   &ParticleState()
   &InheritVelocity(
    Module.Limit Speed=true
    Velocity Limit=10000
    Velocity Scale=&Multiply_VectorByFloat(Vector=<-0.03 -0.03 -0.03>, Float=0.5)
    Velocity Source=&TransformVector(Vector=$Engine.Owner.Velocity, Source Space=::World, Destination Space=::Local)
   )
   &SpriteRotationRate(Rotation Rate=&RandomRangeFloat(Minimum=-90, Maximum=90))
   &Drag(Drag=0.5, Use Linear Drag=true, Use Rotational Drag=false)
   &GravityForce(Gravity=<0 0 -900>)
   &SolveForcesAndVelocity(Clamp Velocity=false, Limit Acceleration=false)
   &ScaleSpriteSize(Scale Factor=&Vector2DFromFloat(Value=&FloatFromCurve(
      FloatCurve=DataInterface() {propertyValues {Curve {keys [
          {arriveTangent=1.6115, interpMode=RCIM_Cubic, leaveTangent=1.6115, tangentMode=RCTM_User, time=0, value=0.33}
          {interpMode=RCIM_Cubic, tangentMode=RCTM_User, time=1, value=1}
         ]}}}
      Scale Curve=0.65
     )))
   &ScaleColor(
    Scale Alpha=&Multiply_Float(
     A=&FloatFromCurve(FloatCurve=DataInterface() {propertyValues {Curve {keys [
          {arriveTangent=9.0725, interpMode=RCIM_Cubic, leaveTangent=9.0725, tangentMode=RCTM_User, time=0, value=0}
          {arriveTangent=-2.5249, interpMode=RCIM_Cubic, leaveTangent=-2.5249, tangentMode=RCTM_User, time=0.1408, value=0.9847}
          <at 0.3 0.2>
          {arriveTangent=-0.0405, interpMode=RCIM_Cubic, leaveTangent=-0.0405, tangentMode=RCTM_User, time=1, value=0}
         ]}}})
     B=0.2
    )
    Scale RGB=&MakeVectorFromLinearColorRGB(Color=&MultiplyLinearColor(
      Linear Color=&ColorFromCurve(ColorCurve=DataInterface() {propertyValues {
         AlphaCurve {keys [<at 0 1 RCIM_Cubic>,<at 1 1 RCIM_Cubic>]}
         BlueCurve {keys [<at 0.1759 1 RCIM_Cubic>,<at 0.2872 1>,<at 0.6925 1 RCIM_Cubic>]}
         GreenCurve {keys [<at 0.1759 0.5996 RCIM_Cubic>,<at 0.2872 0>,<at 0.6925 0.0015 RCIM_Cubic>]}
         RedCurve {keys [<at 0.1759 0 RCIM_Cubic>,<at 0.2872 0.4267>,<at 0.6925 0 RCIM_Cubic>]}
        }})
      Scale Factor=&MakeFloat4FromVectorAndFloat(
       "Vector (RGB)"=&VectorFromFloat(Value=&FloatFromCurve(
         FloatCurve=DataInterface() {propertyValues {Curve {keys [
             {arriveTangent=-2.8172, interpMode=RCIM_Cubic, leaveTangent=-2.8172, tangentMode=RCTM_User, time=0, value=150}
             <at 0.2 20>
             {arriveTangent=-0.0599, interpMode=RCIM_Cubic, leaveTangent=-0.0599, tangentMode=RCTM_User, time=1, value=0.5}
            ]}}}
         Scale Curve=10
        ))
       "Float (Alpha)"=1
      )
     ))
    ScaleA=true
    ScaleRGB=true
   )
   &DynamicMaterialParameters(
    Index 0 Param 1=&RandomRangeFloat(Minimum=0, Maximum=1)
    Index 0 Param 3=&FloatFromCurve(FloatCurve=DataInterface() {propertyValues {Curve {keys [{interpMode=RCIM_Cubic, tangentMode=RCTM_User, time=0, value=0}, {interpMode=RCIM_Cubic, tangentMode=RCTM_User, time=1, value=1}]}}})
    Index 0 Param 4=&FloatFromCurve(FloatCurve=DataInterface() {propertyValues {Curve {keys [
         {arriveTangent=0.2857, interpMode=RCIM_Cubic, leaveTangent=0.2857, tangentMode=RCTM_User, time=0, value=0}
         {interpMode=RCIM_Cubic, tangentMode=RCTM_User, time=1, value=1}
        ]}}})
    Module.Param0WriteEnabled=true
    Module.Param1WriteEnabled=false
    Module.Param2WriteEnabled=true
    Module.Param3WriteEnabled=true
   )
  ]
  rendererClasses [&NiagaraSpriteRendererProperties]
  renderers [{rendererClass=&NiagaraSpriteRendererProperties}]
 }
 Splash_Foam001 {
  emitterUpdateScript [
   &EmitterState(Inactive Response=::NewEnumerator1, Life Cycle Mode=::NewEnumerator1, Loop Duration=1.3, UseLoopDelay=false)
   &SpawnRate(
    Spawn Probability=&FN_FrontEndFloatModifier(Gameplay=0, Front End=1)
    SpawnRate=&FloatFromCurve(FloatCurve=DataInterface() {propertyValues {Curve {keys [<at 0 1>,<at 0.8 0>]}}}, CurveIndex=$Emitter.LoopedAge, Scale Curve=80)
    Use Spawn Probability=true
   ) {module=SpawnRate001}
   &SpawnRate(
    Spawn Probability=&FN_FrontEndFloatModifier(Gameplay=0, Front End=1)
    SpawnRate=10
    Use Spawn Probability=true
   )
  ]
  particleSpawnScript [
   &FN_SetSpawnGroup(Number of Spawn Groups=2) {enabled=false}
   &PartitionParticles()
   /SetVariables(Particles.MyCustomIndex:NiagaraInt32=&SelectIntFromArray(Direct Array Index=$Particles.PartitionParticles.Partition, Int Selection Array=DataInterface() {propertyValues {IntData [2,3]}})) {isSetParameters=true, module=SetVariables_#12}
   /SetVariables(
    Particles.FrontToBack:NiagaraFloat=&RandomRangeFloat(Minimum=0, Maximum=1)
    Particles.LeanLerp:NiagaraFloat=&Multiply_Float(
     A=$User.RightDot
     B=&Lerp_Float(A=-1, B=1, Alpha=&MakeFloatFromInt(Int=$Particles.MyCustomIndex))
    )
   ) {isSetParameters=true, module=SetVariables_#13}
   &InitializeParticle(
    Color=<rgba 1 1 1 0.1>
    Lifetime Max=0.4
    Lifetime Min=0.2
    Lifetime Mode=::NewEnumerator1
    Mesh Scale Mode=::NewEnumerator3
    Sprite Rotation Mode=::NewEnumerator2
    Sprite Size Mode=::NewEnumerator1
    Uniform Sprite Size Max=100
    Uniform Sprite Size Min=75
    UsePositionOffset=false
    Write Material Random=true
   ) {module=InitializeParticle001}
   &SkeletalMeshLocation(
    Apply Sampled Orientation to Mesh Particles=true
    Bone / Socket Index=$Particles.MyCustomIndex
    Inherit Velocity=false
    Interpolate Mesh Sampling=false
    MaskSpawnGroup=false
    Mesh Sampling Type=::NewEnumerator5
    Sampled Position Offset=&VectorFromCurve(
     VectorCurve=DataInterface() {propertyValues {
       XCurve {keys [<at 0 20>, {arriveTangent=-382.5781, interpMode=RCIM_Cubic, leaveTangent=-382.5781, tangentMode=RCTM_User, time=1, value=-120}]}
       YCurve {keys [<at 0.33 0 RCIM_Cubic>, {arriveTangent=-92.7914, interpMode=RCIM_Cubic, leaveTangent=-92.7914, tangentMode=RCTM_User, time=1, value=-35}]}
       ZCurve {keys [<at 0 5>,<at 1 10>]}
      }}
     CurveIndex=$Particles.FrontToBack
     Scale Curve=&Lerp_Vector(A=<-1 0.1 1>, B=<-1 -0.1 1>, Alpha=&MakeFloatFromInt(Int=$Particles.MyCustomIndex))
    )
    Sampled Position Offset Coordinate Space:ENiagaraMeshTransforms=::NewEnumerator3
    Skeletal Mesh=$User.SkeletalMesh
    Socket Sampling Mode=::NewEnumerator5
   ) {module=SkeletalMeshLocation001}
   &SubUVAnimation(SubUV Animation Mode=::NewEnumerator1)
   /SetVariables(Particles.Velocity=&Multiply_VectorByFloat(
     Vector=&ReturnMeshOrientationAxis(Mesh Orientation Axis=::XAxis, Mesh Orientation Quat=$Particles.MeshOrientation)
     Float=&Lerp_Float(A=-400, B=-500, Alpha=$User.ForwardDot)
    )) {isSetParameters=true, module=SetVariables_#14}
   &AddVelocity(
    Scale Added Velocity=&Lerp_Vector(
     A=<1 1 1>
     B=<1.1 1.1 1.25>
     Alpha=&Add_Float(A=&ClampFloat(Float=$User.ForwardDot), B=$Particles.LeanLerp)
    )
    Velocity=&RandomRangeVector(Minimum=<-10 30 175>, Maximum=<10 50 250>)
   )
   &AddVelocityFromPoint(Module.Use Falloff Distance=false, Origin Offset=<0 -20 -15>, Origin Offset Coordinate Space=::Simulation)
  ]
  particleUpdateScript [
   &ParticleState()
   &InheritVelocity(
    Module.Limit Speed=true
    Velocity Limit=10000
    Velocity Scale=&Multiply_VectorByFloat(Vector=<-0.03 -0.03 -0.03>, Float=0.5)
    Velocity Source=&TransformVector(Vector=$Engine.Owner.Velocity, Source Space=::World, Destination Space=::Local)
   )
   &SpriteRotationRate(Rotation Rate=&RandomRangeFloat(Minimum=-90, Maximum=90))
   &Drag(Drag=0.5, Use Linear Drag=true, Use Rotational Drag=false)
   &GravityForce(Gravity=<0 0 -900>)
   &SolveForcesAndVelocity(Clamp Velocity=false, Limit Acceleration=false)
   &ScaleSpriteSize(Scale Factor=&Vector2DFromFloat(Value=&FloatFromCurve(
      FloatCurve=DataInterface() {propertyValues {Curve {keys [
          {arriveTangent=2.2233, interpMode=RCIM_Cubic, leaveTangent=2.2233, tangentMode=RCTM_User, time=0, value=0}
          {interpMode=RCIM_Cubic, tangentMode=RCTM_User, time=1, value=1}
         ]}}}
      Scale Curve=0.65
     )))
   &ScaleColor(
    Scale Alpha=&Multiply_Float(
     A=&FloatFromCurve(FloatCurve=DataInterface() {propertyValues {Curve {keys [
          {arriveTangent=9.0725, interpMode=RCIM_Cubic, leaveTangent=9.0725, tangentMode=RCTM_User, time=0, value=0}
          {arriveTangent=-2.5249, interpMode=RCIM_Cubic, leaveTangent=-2.5249, tangentMode=RCTM_User, time=0.1408, value=0.9847}
          <at 0.3 0.2>
          {arriveTangent=-0.0405, interpMode=RCIM_Cubic, leaveTangent=-0.0405, tangentMode=RCTM_User, time=1, value=0}
         ]}}})
     B=0.2
    )
    Scale RGB=&MakeVectorFromLinearColorRGB(Color=&MultiplyLinearColor(
      Linear Color=&ColorFromCurve(ColorCurve=DataInterface() {propertyValues {
         AlphaCurve {keys [<at 0 1 RCIM_Cubic>,<at 1 1 RCIM_Cubic>]}
         BlueCurve {keys [<at 0.1759 1 RCIM_Cubic>,<at 0.2872 1>,<at 0.6925 1 RCIM_Cubic>]}
         GreenCurve {keys [<at 0.1759 0.5996 RCIM_Cubic>,<at 0.2872 0>,<at 0.6925 0.0015 RCIM_Cubic>]}
         RedCurve {keys [<at 0.1759 0 RCIM_Cubic>,<at 0.2872 0.4267>,<at 0.6925 0 RCIM_Cubic>]}
        }})
      Scale Factor=&MakeFloat4FromVectorAndFloat(
       "Vector (RGB)"=&VectorFromFloat(Value=&FloatFromCurve(
         FloatCurve=DataInterface() {propertyValues {Curve {keys [
             {arriveTangent=-2.8172, interpMode=RCIM_Cubic, leaveTangent=-2.8172, tangentMode=RCTM_User, time=0, value=150}
             <at 0.2 20>
             {arriveTangent=-0.0599, interpMode=RCIM_Cubic, leaveTangent=-0.0599, tangentMode=RCTM_User, time=1, value=0.5}
            ]}}}
         Scale Curve=10
        ))
       "Float (Alpha)"=1
      )
     ))
    ScaleA=true
    ScaleRGB=true
   )
   &DynamicMaterialParameters(
    Index 0 Param 1=&RandomRangeFloat(Minimum=0, Maximum=1)
    Index 0 Param 3=&FloatFromCurve(FloatCurve=DataInterface() {propertyValues {Curve {keys [{interpMode=RCIM_Cubic, tangentMode=RCTM_User, time=0, value=0}, {interpMode=RCIM_Cubic, tangentMode=RCTM_User, time=1, value=1}]}}})
    Index 0 Param 4=&FloatFromCurve(FloatCurve=DataInterface() {propertyValues {Curve {keys [
         {arriveTangent=0.2857, interpMode=RCIM_Cubic, leaveTangent=0.2857, tangentMode=RCTM_User, time=0, value=0}
         {arriveTangent=1.9495, interpMode=RCIM_Cubic, leaveTangent=1.9495, tangentMode=RCTM_User, time=1, value=1}
        ]}}})
    Module.Param0WriteEnabled=true
    Module.Param1WriteEnabled=false
    Module.Param2WriteEnabled=true
    Module.Param3WriteEnabled=true
   )
  ]
  rendererClasses [&NiagaraSpriteRendererProperties]
  renderers [{rendererClass=&NiagaraSpriteRendererProperties}]
 }
 Trail_Blue {
  emitterSpawnScript [&FN_RibbonIDOffsetUtility()]
  emitterUpdateScript [
   &EmitterState(Life Cycle Mode=::NewEnumerator1, Loop Delay=0.1, Loop Duration=1, UseLoopDelay=false)
   &FN_SpawnAnimTrail(Number of Anim Trails:NiagaraInt32=&MakeIntFromFloat(Float=&FN_FrontEndFloatModifier(Gameplay=1, Front End=0)))
  ]
  particleSpawnScript [
   &FN_InitializeAnimTrailBoneIndecies(First Bone or Socket Index=3, Second Bone or Socket Index=2)
   &FN_InitializeAnimTrailBoneIndecies(AnimTrail Selection=1, First Bone or Socket Index=2, Second Bone or Socket Index=3) {module=FN_InitializeAnimTrailBoneIndecies001}
   /SetVariables(Particles.AlphaFadeIn:NiagaraFloat=&FloatFromCurve(FloatCurve=DataInterface() {propertyValues {Curve {keys [<at 2 0>,<at 3 1>]}}}, CurveIndex=$Emitter.Age)) {isSetParameters=true, module=SetVariables_#15}
   &FN_InitializeAnimTrail(
    AnimTrailWidth=2
    Color=&MakeLinearColorFromVectorAndFloat("Vector (RGB)"=&VectorFromFloat(Value=1.3), "Float (Alpha)"=1.5)
    Lifetime=&MakeCustomFloatFromBool(Bool=$Particles.Locked, True Float=100, False Float=2)
    Position Offset=&FN_FrontEndVector3Modifier(Gameplay=<0 -150 10>, Front End=<0 30 0>)
   )
   &AddVelocity(Velocity=&Multiply_VectorByFloat(Vector=<0 400 0>, Float=&MakeCustomFloatFromBool(Bool=$User.bIsFrontEnd, True Float=-1, False Float=1)))
  ]
  particleUpdateScript [
   &SolveForcesAndVelocity(Clamp Velocity=false, Limit Acceleration=false)
   &FN_UpdateAnimTrail(
    Locked Particle Color:LinearColor=&MultiplyLinearColor(Linear Color=$Particles.Initial.Color, Scale Factor=<1 1 1 1>)
    Locked Particle RibbonWidth=1
    SkeletalMesh=$User.SkeletalMesh
   ) {module=FN_UpdateAnimTrail001}
   /SetVariables(Particles.RibbonU0Override=$Particles.NormalizedAge) {isSetParameters=true, module=SetVariables_#16}
   &Material(Base Texture Time:NiagaraFloat=&Multiply_Float(A=$Emitter.Age, B=-0.5), U Strength=0.5) {module=Material002}
   &ScaleColor(
    Scale Alpha=&Multiply_Float(
     A=&Multiply_Float(
      A=&FloatFromCurve(FloatCurve=DataInterface() {propertyValues {Curve {keys [
           {interpMode=RCIM_Cubic, tangentMode=RCTM_User, time=0.01, value=1}
           {interpMode=RCIM_Cubic, tangentMode=RCTM_User, time=0.8, value=1}
           {tangentMode=RCTM_User, time=1, value=0}
          ]}}})
      B=$Particles.AlphaFadeIn
     )
     B=&FN_FrontEndFloatModifier(
      Gameplay=&Max_Float(A:NiagaraFloat=$System.Timeline.NormalizedTime, B:NiagaraFloat=$System.Timeline002.NormalizedTime)
      Front End=1
     )
    )
    Scale RGB=&FN_FrontEndVector3Modifier(
     Gameplay=&MakeVectorFromLinearColorRGB(Color=&ColorFromCurve(ColorCurve=DataInterface() {propertyValues {
         AlphaCurve {keys [<at 0 1 RCIM_Cubic>,<at 1 1 RCIM_Cubic>]}
         BlueCurve {keys [<at 0.3014 0.5>,<at 0.5243 0.4>,<at 0.6686 0.1>]}
         GreenCurve {keys [<at 0.3014 0.1571>,<at 0.5243 0.0736>,<at 0.6686 0.0257>]}
         RedCurve {keys [<at 0.3014 0.05>,<at 0.5243 0.044>,<at 0.6686 0.01>]}
        }}))
     Front End=&MakeVectorFromLinearColorRGB(Color=&ColorFromCurve(ColorCurve=DataInterface() {propertyValues {
         AlphaCurve {keys [<at 0 1 RCIM_Cubic>,<at 1 1 RCIM_Cubic>]}
         BlueCurve {keys [<at 0 0 RCIM_Cubic>,<at 0.0496 0>,<at 0.2189 0.2462>,<at 0.681 0.2>,<at 1 1 RCIM_Cubic>]}
         GreenCurve {keys [<at 0 0 RCIM_Cubic>,<at 0.0496 0>,<at 0.2189 0.0417>,<at 0.681 0.072>,<at 1 1 RCIM_Cubic>]}
         RedCurve {keys [<at 0 0 RCIM_Cubic>,<at 0.0496 0>,<at 0.2189 0>,<at 0.681 0.02>,<at 1 1 RCIM_Cubic>]}
        }}))
    )
    ScaleA=true
    ScaleRGB=true
   ) {module=ScaleColor001}
   /SetVariables(Particles.RibbonWidth=&FloatFromCurve(
     FloatCurve=DataInterface() {propertyValues {Curve {keys [
         {arriveTangent=0.573, interpMode=RCIM_Cubic, leaveTangent=0.573, tangentMode=RCTM_User, time=0, value=0.5}
         {arriveTangent=1.25, interpMode=RCIM_Cubic, leaveTangent=1.25, time=0.2343, value=1.0246}
         {arriveTangent=-1.3382, interpMode=RCIM_Cubic, leaveTangent=-1.3382, time=0.5, value=1.125}
         {arriveTangent=-2.3973, interpMode=RCIM_Cubic, leaveTangent=-2.3973, tangentMode=RCTM_User, time=1, value=0}
        ]}}}
     Scale Curve=160
    )) {isSetParameters=true, module=SetVariables_#17}
   /SetVariables(Particles.RibbonWidth2=50) {enabled=false, isSetParameters=true, module=SetVariables_#18}
   &DynamicMaterialParameters(
    Index 0 Param 1=0.94
    Index 0 Param 2=&Multiply_Float(
     A=&FloatFromCurve(FloatCurve=DataInterface() {propertyValues {Curve {keys [
          {arriveTangent=9.0725, interpMode=RCIM_Cubic, leaveTangent=9.0725, tangentMode=RCTM_User, time=0.1, value=0}
          {arriveTangent=-2.5249, interpMode=RCIM_Cubic, leaveTangent=-2.5249, tangentMode=RCTM_User, time=0.2, value=0.9847}
          {arriveTangent=-0.0405, interpMode=RCIM_Cubic, leaveTangent=-0.0405, tangentMode=RCTM_User, time=1, value=0}
         ]}}})
     B=&Multiply_Float(
      A=&Multiply_Float(A=1, B=$Particles.AlphaFadeIn)
      B=&FN_FrontEndFloatModifier(
       Gameplay=&Max_Float(A:NiagaraFloat=$System.Timeline.NormalizedTime, B:NiagaraFloat=$System.Timeline002.NormalizedTime)
       Front End=0.1
      )
     )
    )
    Module.Param0WriteEnabled=true
    Module.Param1WriteEnabled=true
    Module.Param2WriteEnabled=true
    Module.Param3WriteEnabled=true
   )
  ]
  rendererClasses [&NiagaraRibbonRendererProperties]
  renderers [{rendererClass=&NiagaraRibbonRendererProperties}]
 }
 Trail_Blue001 {
  emitterSpawnScript [&FN_RibbonIDOffsetUtility()]
  emitterUpdateScript [
   &EmitterState(Life Cycle Mode=::NewEnumerator1, Loop Delay=0.1, Loop Duration=1, UseLoopDelay=false)
   &FN_SpawnAnimTrail(Number of Anim Trails:NiagaraInt32=&MakeIntFromFloat(Float=&FN_FrontEndFloatModifier(Gameplay=0, Front End=1)))
  ]
  particleSpawnScript [
   &FN_InitializeAnimTrailBoneIndecies(First Bone or Socket Index=3, Second Bone or Socket Index=2)
   &FN_InitializeAnimTrailBoneIndecies(AnimTrail Selection=1, First Bone or Socket Index=2, Second Bone or Socket Index=3) {module=FN_InitializeAnimTrailBoneIndecies001}
   /SetVariables(Particles.AlphaFadeIn:NiagaraFloat=&FloatFromCurve(FloatCurve=DataInterface() {propertyValues {Curve {keys [<at 2 0>, {tangentMode=RCTM_User, time=3, value=1}]}}}, CurveIndex=$Emitter.Age)) {isSetParameters=true, module=SetVariables_#15}
   &FN_InitializeAnimTrail(
    AnimTrailWidth=2
    Color=&MakeLinearColorFromVectorAndFloat("Vector (RGB)"=&VectorFromFloat(Value=1.3), "Float (Alpha)"=0.5)
    Lifetime=&MakeCustomFloatFromBool(Bool=$Particles.Locked, True Float=100, False Float=2)
    Position Offset=&FN_FrontEndVector3Modifier(Gameplay=<0 -150 0>, Front End=<0 30 0>)
   )
   &AddVelocity(Velocity=&Multiply_VectorByFloat(Vector=<0 400 0>, Float=&MakeCustomFloatFromBool(Bool=$User.bIsFrontEnd, True Float=1, False Float=1)))
  ]
  particleUpdateScript [
   &SolveForcesAndVelocity(Clamp Velocity=false, Limit Acceleration=false)
   &FN_UpdateAnimTrail(
    Locked Particle Color:LinearColor=&MultiplyLinearColor(Linear Color=$Particles.Initial.Color, Scale Factor=<1 1 1 1>)
    Locked Particle RibbonWidth=1
    SkeletalMesh=$User.SkeletalMesh
   ) {module=FN_UpdateAnimTrail001}
   /SetVariables(Particles.RibbonU0Override=$Particles.NormalizedAge) {isSetParameters=true, module=SetVariables_#16}
   &Material(Base Texture Time:NiagaraFloat=&Multiply_Float(A=$Emitter.Age, B=-0.5), U Strength=0.5) {module=Material002}
   &ScaleColor(
    Scale Alpha=&Multiply_Float(
     A=&Multiply_Float(
      A=&FloatFromCurve(FloatCurve=DataInterface() {propertyValues {Curve {keys [
           {interpMode=RCIM_Cubic, tangentMode=RCTM_User, time=0.01, value=1}
           {interpMode=RCIM_Cubic, tangentMode=RCTM_User, time=0.8, value=1}
           {tangentMode=RCTM_User, time=1, value=0}
          ]}}})
      B=$Particles.AlphaFadeIn
     )
     B=&FN_FrontEndFloatModifier(
      Gameplay=&Max_Float(A:NiagaraFloat=$System.Timeline.NormalizedTime, B:NiagaraFloat=$System.Timeline002.NormalizedTime)
      Front End=1
     )
    )
    Scale RGB=&FN_FrontEndVector3Modifier(
     Gameplay=&MakeVectorFromLinearColorRGB(Color=&ColorFromCurve(ColorCurve=DataInterface() {propertyValues {
         AlphaCurve {keys [<at 0 1 RCIM_Cubic>,<at 1 1 RCIM_Cubic>]}
         BlueCurve {keys [<at 0.3014 0.5>,<at 0.5243 0.4>,<at 0.6686 0.1>]}
         GreenCurve {keys [<at 0.3014 0.1571>,<at 0.5243 0.0736>,<at 0.6686 0.0257>]}
         RedCurve {keys [<at 0.3014 0.05>,<at 0.5243 0.044>,<at 0.6686 0.01>]}
        }}))
     Front End=&MakeVectorFromLinearColorRGB(Color=&ColorFromCurve(ColorCurve=DataInterface() {propertyValues {
         AlphaCurve {keys [<at 0 1 RCIM_Cubic>,<at 1 1 RCIM_Cubic>]}
         BlueCurve {keys [<at 0.1507 1>,<at 0.4182 1>,<at 0.6347 0.2>,<at 0.987 0.1>]}
         GreenCurve {keys [<at 0.1507 0.3143>,<at 0.4182 0.1015>,<at 0.6347 0.0629>,<at 0.987 0.0257>]}
         RedCurve {keys [<at 0.1507 0.1>,<at 0.4182 0.02>,<at 0.6347 0.02>,<at 0.987 0.01>]}
        }}))
    )
    ScaleA=true
    ScaleRGB=true
   ) {module=ScaleColor001}
   /SetVariables(Particles.RibbonWidth=&FloatFromCurve(
     FloatCurve=DataInterface() {propertyValues {Curve {keys [
         {arriveTangent=0.573, interpMode=RCIM_Cubic, leaveTangent=0.573, tangentMode=RCTM_User, time=0, value=0.5}
         {arriveTangent=1.25, interpMode=RCIM_Cubic, leaveTangent=1.25, time=0.2343, value=1.0246}
         {arriveTangent=-1.3382, interpMode=RCIM_Cubic, leaveTangent=-1.3382, time=0.5, value=1.125}
         {arriveTangent=-2.3973, interpMode=RCIM_Cubic, leaveTangent=-2.3973, tangentMode=RCTM_User, time=1, value=0}
        ]}}}
     Scale Curve=160
    )) {isSetParameters=true, module=SetVariables_#17}
   /SetVariables(Particles.RibbonWidth2=50) {enabled=false, isSetParameters=true, module=SetVariables_#18}
   &DynamicMaterialParameters(
    Index 0 Param 1=1
    Index 0 Param 2=&Multiply_Float(
     A=&FloatFromCurve(FloatCurve=DataInterface() {propertyValues {Curve {keys [
          {arriveTangent=9.0725, interpMode=RCIM_Cubic, leaveTangent=9.0725, tangentMode=RCTM_User, time=0, value=0}
          {arriveTangent=-2.5249, interpMode=RCIM_Cubic, leaveTangent=-2.5249, tangentMode=RCTM_User, time=0.1408, value=0.9847}
          {arriveTangent=-0.0405, interpMode=RCIM_Cubic, leaveTangent=-0.0405, tangentMode=RCTM_User, time=1, value=0}
         ]}}})
     B=&Multiply_Float(
      A=&Multiply_Float(A=3, B=$Particles.AlphaFadeIn)
      B=&FN_FrontEndFloatModifier(
       Gameplay=&Max_Float(A:NiagaraFloat=$System.Timeline.NormalizedTime, B:NiagaraFloat=$System.Timeline002.NormalizedTime)
       Front End=2
      )
     )
    )
    Module.Param0WriteEnabled=true
    Module.Param1WriteEnabled=true
    Module.Param2WriteEnabled=true
    Module.Param3WriteEnabled=true
   )
  ]
  rendererClasses [&NiagaraRibbonRendererProperties]
  renderers [{rendererClass=&NiagaraRibbonRendererProperties}]
 }
}
summary {
 systemName=NS_Glider_CyberFuGlitch
 emitters [
  {emitterName=Splash_Drops, rendererClasses [&NiagaraSpriteRendererProperties]}
  {emitterName=Trail_Blue, rendererClasses [&NiagaraRibbonRendererProperties]}
  {emitterName=Speedlines, rendererClasses [&NiagaraSpriteRendererProperties]}
  {bEnabled=false, emitterName=ExampleOrlandoSpawnGroups, rendererClasses [&NiagaraSpriteRendererProperties]}
  {emitterName=Ripples_Deploy, rendererClasses [&NiagaraSpriteRendererProperties, &NiagaraMeshRendererProperties]}
  {
   emitterName=MeshRipplesMainLocal
   rendererClasses [&NiagaraSpriteRendererProperties, &NiagaraMeshRendererProperties]
  }
  {emitterName=Splash_Foam, rendererClasses [&NiagaraSpriteRendererProperties]}
  {emitterName=Splash_Drops001, rendererClasses [&NiagaraSpriteRendererProperties]}
  {emitterName=Trail_Blue001, rendererClasses [&NiagaraRibbonRendererProperties]}
  {
   emitterName=Ripples_Deploy001
   rendererClasses [&NiagaraSpriteRendererProperties, &NiagaraMeshRendererProperties]
  }
  {emitterName=Splash_Foam001, rendererClasses [&NiagaraSpriteRendererProperties]}
  {emitterName=MeshRipples_FE, rendererClasses [&NiagaraSpriteRendererProperties, &NiagaraMeshRendererProperties]}
 ]
 userVariables [
  User.bIsFrontEnd {type=NiagaraBool}
  User.bIsFrontEndPreview {type=NiagaraBool}
  User.bIsFullyDeployed {type=NiagaraBool}
  User.BPOpacityMult {type=NiagaraFloat, defaultValue=0, description=""}
  User.CenterScale {type=NiagaraFloat, defaultValue=1, description=""}
  User.CurrentExhaust {type=Vector3f, description="", defaultValue {struct=Vector3f, value=<0 0 0>}}
  User.ExhaustAngle {type=NiagaraFloat, defaultValue=0, description=""}
  User.ForwardDot {type=NiagaraFloat, defaultValue=0, description=""}
  User.PlayerSpeed {type=NiagaraFloat, defaultValue=0, description=""}
  User.RibbonIDOffset {type=NiagaraInt32}
  User.RightDot {type=NiagaraFloat, defaultValue=0, description=""}
  User.SkeletalMesh {
   type=NiagaraDataInterfaceSkeletalMesh
   defaultValue {
    struct=NiagaraExt_VariableValue_DataInterface
    value {dataInterface=&self:NiagaraDataInterfaceSkeletalMesh_1}
   }
  }
  User.TrailLengthScale {type=NiagaraFloat, defaultValue=0, description=""}
 ]
}
